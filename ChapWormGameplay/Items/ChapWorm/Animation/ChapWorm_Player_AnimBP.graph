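# ChapWorm player animation Blueprint that fills item override layers with Kinetic Energy animations.
# Pre-IK state machine adds a looping idle additive, masked off the lower body when crouching.
# Final-pose state machine plays a one-shot activation from Spine_02, restores the neck upward in mesh space, and uses inertialized transitions through Dead Blending.
# A surface-swimming right-hand finger blend exists but is unconnected; other layers have no logic.
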AnimBlueprint /ChapWormGameplay/Items/ChapWorm/Animation/ChapWorm_Player_AnimBP
target_skeleton=&Fortnite_M_Avg_Player_Skeleton
graphs [
 AnimGraph {nodes [Output Pose()]}
 EventGraph {links [], nodes []}
 ItemAimOffset {nodes [
   Input Pose!() -> (Pose, AimOffsetAlpha:double, Yaw:double, Pitch:double) {name=InPose_AimOffset}
   Input Pose!() -> (Pose) {name=InPose_PreAimOffsets}
   Output Pose() {layerGroup=Item Override}
  ]}
 ItemFallAdditive {nodes [
   Input Pose() {name=InPoseItemFallAdditive}
   Output Pose() {layerGroup=Item Override}
  ]}
 ItemFallLandAdditive {nodes [
   Input Pose() {name=InPoseItemFallLandAdditive}
   Output Pose() {layerGroup=Item Override}
  ]}
 ItemFallLevel2Override {nodes [
   Input Pose() {name=InPose_FallLevel2}
   Output Pose() {layerGroup=Item Override}
  ]}
 ItemFinalPoseOverride {nodes [
   AnimGraphNode_StateMachine_0 = FinalPose SM() {is=title AnimGraphNode_StateMachine}
   Output Pose(Result=Dead Blending!(Source:Pose Link Structure=@AnimGraphNode_StateMachine_0) -> (Pose:Pose Link Structure) {is=title AnimGraphNode_DeadBlending, defaultBlendMode=Linear}) {layerGroup=Item Override}
   Save cached pose 'FinalPoseCache'(Pose=Input Pose() {name=InFinalPose}) {cacheName=FinalPoseCache}
  ]}
 FinalPose SM {
  entry=Base
  path=ItemFinalPoseOverride.AnimGraphNode_StateMachine_0.FinalPose SM
  ids=id
  nodes [
   AnimStateNode_0 = Base() {is=title AnimStateNode}
   AnimStateNode_1 = Activate() {is=title AnimStateNode}
   AnimStateTransitionNode_0 = AnimStateTransitionNode() {automaticRuleTriggerTime=0, crossfadeDuration=0.1, logicType=TLT_Inertialization}
   AnimStateTransitionNode_1 = AnimStateTransitionNode() {automaticRuleTriggerTime=0, bAutomaticRuleBasedOnSequencePlayerInState=true, logicType=TLT_Inertialization}
  ]
  links [
   {
    dst=Base
    rule=ItemFinalPoseOverride.AnimGraphNode_StateMachine_0.FinalPose SM.AnimStateTransitionNode_1.Transition
    src=Activate
    via=AnimStateTransitionNode_1
   }
   {
    dst=Activate
    rule=ItemFinalPoseOverride.AnimGraphNode_StateMachine_0.FinalPose SM.AnimStateTransitionNode_0.Transition
    src=Base
    via=AnimStateTransitionNode_0
   }
  ]
 }
 Base {
  path=ItemFinalPoseOverride.AnimGraphNode_StateMachine_0.FinalPose SM.AnimStateNode_0.Base
  nodes [Output Animation Pose(Result=Use cached pose 'FinalPoseCache'())]
 }
 Activate {
  path=ItemFinalPoseOverride.AnimGraphNode_StateMachine_0.FinalPose SM.AnimStateNode_1.Activate
  nodes [
   n1 = Sequence Player 'ChapWorm_KineticEnergy_Activate'() {bLoopAnimation=false, sequence=/ChapWormGameplay/Items/ChapWorm/Animation/KineticEnergy/ChapWorm_KineticEnergy_Activate}
   n2 = Use cached pose 'FinalPoseCache'()
   Output Animation Pose(Result=Layered blend per bone(
     BasePose=Layered blend per bone(BasePose=@n2, BlendPoses_0=@n1) {layerSetup [{branchFilters [{blendDepth=4, boneName=Spine_02}]}]}
     BlendPoses_0=Use cached pose 'FinalPoseCache'()
    ) {bMeshSpaceRotationBlend=true, layerSetup [{branchFilters [{boneName=neck_01}]}]})
  ]
 }
 Transition {
  path=ItemFinalPoseOverride.AnimGraphNode_StateMachine_0.FinalPose SM.AnimStateTransitionNode_0.Transition
  nodes [Result(bCanEnterTransition=Get bCanEnterBaseToActivate() -> (bCanEnterBaseToActivate:bool))]
 }
 Transition {path=ItemFinalPoseOverride.AnimGraphNode_StateMachine_0.FinalPose SM.AnimStateTransitionNode_1.Transition, nodes [Result()]}
 ItemFullBodyAimOffsets {nodes [
   Input Pose!() -> (Pose, FullBody_AimOffsetAlpha:double, Aim_Yaw:double, Aim_Pitch:double) {name=InPoseFullBodyAimOffsets}
   Input Pose!() -> (Pose) {name=InPosePreFullBodyAimOffsets}
   Output Pose() {layerGroup=Item Override}
  ]}
 ItemFullBodyOverride {nodes [
   Layered blend per bone(
    BlendPoses_0=Sequence Player 'ChapWorm_Swim_Idle_NonTargeted_Pose'() {bLoopAnimation=false, sequence=&/ChapWorm_Swim_Idle_NonTargeted_Pose}
    BlendWeights_0="To Float (Boolean)"(InBool=Get bIsSurfaceSwimming())
   ) {layerSetup [{branchFilters [
       {boneName=index_metacarpal_r}
       {boneName=middle_metacarpal_r}
       {boneName=pinky_metacarpal_r}
       {boneName=ring_metacarpal_r}
       {boneName=thumb_01_r}
      ]}]}
   Input Pose() {name=InPoseFullBody}
   Output Pose() {layerGroup=Item Override}
  ]}
 ItemIdleOverride {nodes [
   Input Pose() {name=InPose_Idle}
   Output Pose() {layerGroup=Item Override}
  ]}
 ItemJetPackJumpAdditive {nodes [Output Pose() {layerGroup=Item Override}]}
 ItemJetPackStartAdditive {nodes [Output Pose() {layerGroup=Item Override}]}
 ItemJumpLoopAdditive {nodes [
   Input Pose() {name=InPoseItemJumpLoopAdditive}
   Output Pose() {layerGroup=Item Override}
  ]}
 ItemJumpOverride {nodes [
   Input Pose() {name=InPoseJump}
   Output Pose() {layerGroup=Item Override}
  ]}
 ItemJumpUpAdditive {nodes [
   Input Pose() {name=InPoseItemJumpUpAdditive}
   Output Pose() {layerGroup=Item Override}
  ]}
 ItemPreAimOffsets {nodes [
   Input Pose() {name=InPosePreAimOffsets}
   Output Pose() {layerGroup=Item Override}
  ]}
 ItemPreIK {nodes [
   AnimGraphNode_StateMachine_0 = KineticEnergy SM() {is=title AnimGraphNode_StateMachine}
   Output Pose(Result=@AnimGraphNode_StateMachine_0) {layerGroup=Item Override}
   Save cached pose 'PreIK Pose'(Pose=Input Pose() {name=InPose_PreIK}) {cacheName=PreIK Pose}
  ]}
 KineticEnergy SM {
  entry=Base
  path=ItemPreIK.AnimGraphNode_StateMachine_0.KineticEnergy SM
  ids=id
  nodes [
   AnimStateNode_0 = Base()
   AnimStateNode_2 = Kinetic Energy Idle() {is=title AnimStateNode}
   AnimStateTransitionNode_2 = AnimStateTransitionNode() {crossfadeDuration=0.3}
   AnimStateTransitionNode_3 = AnimStateTransitionNode() {crossfadeDuration=0.3}
  ]
  links [
   {
    dst=Kinetic Energy Idle
    rule=ItemPreIK.AnimGraphNode_StateMachine_0.KineticEnergy SM.AnimStateTransitionNode_2.Transition
    src=Base
    via=AnimStateTransitionNode_2
   }
   {
    dst=Base
    rule=ItemPreIK.AnimGraphNode_StateMachine_0.KineticEnergy SM.AnimStateTransitionNode_3.Transition
    src=Kinetic Energy Idle
    via=AnimStateTransitionNode_3
   }
  ]
 }
 Base {
  path=ItemPreIK.AnimGraphNode_StateMachine_0.KineticEnergy SM.AnimStateNode_0.Base
  nodes [Output Animation Pose(Result=Use cached pose 'PreIK Pose'())]
 }
 Kinetic Energy Idle {
  path=ItemPreIK.AnimGraphNode_StateMachine_0.KineticEnergy SM.AnimStateNode_2.Kinetic Energy Idle
  nodes [
   n1 = "Sequence Player (additive) 'ChapWorm_KineticEnergy_Idle_Loop_Add'"() {sequence=/ChapWormGameplay/Items/ChapWorm/Animation/KineticEnergy/ChapWorm_KineticEnergy_Idle_Loop_Add}
   n2 = Use cached pose 'PreIK Pose'()
   n3 = Layered blend per bone(
    BasePose=Apply Additive(
     Base=Use cached pose 'PreIK Pose'()
     Additive="Sequence Player (additive) 'ChapWorm_KineticEnergy_Idle_Loop_Add'"() {sequence=&ChapWorm_KineticEnergy_Idle_Loop_Add}
    ) {alphaBoolBlend {blendOption=Linear}}
    BlendPoses_0=Use cached pose 'PreIK Pose'()
   ) {blendMode=BlendMask, blendMasks [&Fortnite_M_Avg_Player_Skeleton:LowerBody], layerSetup []}
   Output Animation Pose(Result=Blend Poses by bool(
     BlendPose_0=@n3
     BlendPose_1=Apply Additive(Base=@n2, Additive=@n1) {alphaBoolBlend {blendOption=Linear}}
     BlendTime_0=0.2
     BlendTime_1=0.2
     bActiveValue=Get bIsCrouching()
    ) {blendType=Linear})
  ]
 }
 Transition {
  path=ItemPreIK.AnimGraphNode_StateMachine_0.KineticEnergy SM.AnimStateTransitionNode_2.Transition
  nodes [Result(bCanEnterTransition=Get bIsKineticEnergyIdle() -> (bIsKineticEnergyIdle:bool))]
 }
 Transition {
  path=ItemPreIK.AnimGraphNode_StateMachine_0.KineticEnergy SM.AnimStateTransitionNode_3.Transition
  nodes [Result(bCanEnterTransition=NOT Boolean(A=Get bIsKineticEnergyIdle()))]
 }
 ItemRightHandIK {nodes [
   Input Pose() {name=InPosePreRightHandIK}
   Input Pose() {name=InPosePostRightHandIK}
   Output Pose() {layerGroup=Item Override}
  ]}
 ItemSkeletalControl {nodes [
   Input Pose()
   Output Pose() {layerGroup=Item Override}
  ]}
 ItemSwimJumpFallLoop {nodes [Output Pose() {layerGroup=Item Override}]}
 ItemSwimJumpSurfaceEnd {nodes [Output Pose() {layerGroup=Item Override}]}
 ItemSwimJumpSurfaceLoop {nodes [Output Pose() {layerGroup=Item Override}]}
 ItemUpperBody {nodes [
   Input Pose!() -> (Pose, InputParam:Fort Anim Input Adjusted Aim Structure) {name=InPoseUpperBody}
   Output Pose() {layerGroup=Item Override}
  ]}
 ItemZipLineStartAdditive {nodes [Output Pose() {layerGroup=Item Override}]}
]
variables []
