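# Looping stereo MetaSound for the ChapWorm's overpowered state, using melee weapon sound class and close-range attenuation.
# On play, an ADSR envelope starts and triggers a random pick from a one-entry wave array into a stereo wave player.
# Movement speed from the Fort.Movement interface raises the player's pitch shift and, through a second range map, boosts gain.
# Both channels are multiplied by the envelope and that gain before output.
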MetaSoundSource /ChapWormGameplay/Items/ChapWorm/Audio/MSS_ChapWorm_Overpowered_lp
inputs [
 In Array {type=WaveAsset:Array}
 Movement.Speed {type=Float}
 Movement.SpeedAzimuth {type=Float}
 Movement.SpeedElevation {type=Float}
 UE.Source.OnPlay {type=Trigger}
]
interfaces [Fort.Movement:1.0, UE.OutputFormat.Stereo:1.0, UE.Source:1.1]
outputs [
 "UE.OutputFormat.Stereo.Audio:0" {type=Audio}
 "UE.OutputFormat.Stereo.Audio:1" {type=Audio}
]
pages [Default {
  input_defaults {
   Movement.Speed=0
   Movement.SpeedAzimuth=0
   Movement.SpeedElevation=0
   UE.Source.OnPlay=null
   In Array [/ChapWormGameplay/Items/ChapWorm/Audio/wav/ChapWorm_Overpowered_lp]
  }
  nodes [
   n1 = MapRange.MapRange:Float(In=^Movement.Speed) {also {defaults {In Range B=600, Out Range A=1, Out Range B=1.4}}}
   ADSREnvelopeADSREnvelope_0 = ADSR Envelope.ADSR Envelope:Audio(Trigger Attack=^UE.Source.OnPlay) {also {defaults {Attack Time=1, Decay Time=0.5, Sustain Level=0.4}}}
   ArrayRandomGet_0 = Array.Random Get:WaveAsset:Array(In Array=^In Array, Next=@ADSREnvelopeADSREnvelope_0.On Attack Triggered)
   WavePlayer_0 = UE.Wave Player:Stereo(
    Pitch Shift=UE.Add:Float(AdditionalOperands=MapRange.MapRange:Float(In=^Movement.Speed) {also {defaults {In Range B=530, Out Range B=5}}}.Out Value) {also {defaults {PrimaryOperand=1}}}.Out
    Play=@ArrayRandomGet_0.On Next
    Wave Asset=@ArrayRandomGet_0.Value
   ) {also {defaults {Loop="<MetaSoundBuilderResult.SUCCEEDED: 0>"}}}
   Multiply_2 = UE.Multiply:Audio by Float(
    AdditionalOperands=@n1.Out Value
    PrimaryOperand=UE.Multiply:Audio(AdditionalOperands=@ADSREnvelopeADSREnvelope_0.Out Envelope, PrimaryOperand=@WavePlayer_0.Out Left).Out
   )
   Multiply_3 = UE.Multiply:Audio by Float(
    AdditionalOperands=@n1.Out Value
    PrimaryOperand=UE.Multiply:Audio(AdditionalOperands=@ADSREnvelopeADSREnvelope_0.Out Envelope, PrimaryOperand=@WavePlayer_0.Out Right).Out
   )
  ]
  links [{dst="@UE.OutputFormat.Stereo.Audio:0", src=Multiply_2.Out}, {dst="@UE.OutputFormat.Stereo.Audio:1", src=Multiply_3.Out}]
 }]
properties {
 assetImportData=&self:AssetImportData
 attenuationSettings=/ChapWormGameplay/Items/ChapWorm/Audio/Settings/CS_ChapWorm_Melee_Attack_Close_Range_Stereo_Att
 bLooping=true
 duration=10000
 maxDistance=7000
 numChannels=2
 outputFormat=Stereo
 pitch=0.8
 sampleRate=48000
 soundClassObject=&Fort_WeaponMelee
}
