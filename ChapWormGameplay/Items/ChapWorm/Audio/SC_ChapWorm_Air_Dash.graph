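# Sound cue for the ChapWorm item's air dash.
# A Random node picks one of three air dash wave variations with equal weight and feeds a Modulator at the root.
# It plays on the melee weapon sound class with a close-range stereo attenuation borrowed from the melee attack settings, a volume boost, and a concurrency override limiting voice count and retriggering.
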SoundCue /ChapWormGameplay/Items/ChapWorm/Audio/SC_ChapWorm_Air_Dash
root=Modulator_0
nodes [
 {id=Modulator_0, children [Random_1]}
 {id=Random_1, children [WavePlayer_0,WavePlayer_1,WavePlayer_2], properties {weights [1,1,1]}}
 {id=WavePlayer_0, properties {soundWaveAssetPtr=/ChapWormGameplay/Items/ChapWorm/Audio/wav/ChapWorm_Air_Dash_01}}
 {id=WavePlayer_1, properties {soundWaveAssetPtr=/ChapWormGameplay/Items/ChapWorm/Audio/wav/ChapWorm_Air_Dash_02}}
 {id=WavePlayer_2, properties {soundWaveAssetPtr=/ChapWormGameplay/Items/ChapWorm/Audio/wav/ChapWorm_Air_Dash_03}}
]
properties {
 attenuationSettings=/ChapWormGameplay/Items/ChapWorm/Audio/Settings/CS_ChapWorm_Melee_Attack_Close_Range_Stereo_Att
 bOverrideConcurrency=true
 maxDistance=7000
 soundClassObject=&Fort_WeaponMelee_PSM
 volumeMultiplier=2
 concurrencyOverrides {maxCount=3, retriggerTime=0.3}
}
