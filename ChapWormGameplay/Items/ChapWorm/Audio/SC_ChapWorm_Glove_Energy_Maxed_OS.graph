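# One-shot SoundCue for the ChapWorm glove, named for the moment its energy reaches maximum.
# A Modulator wraps a single wave player for the Energy Maxed wave.
# It uses the ChapWorm close-range stereo melee attenuation and routes to the Fortnite melee weapon sound class.
# A concurrency override caps simultaneous instances at a small count.
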SoundCue /ChapWormGameplay/Items/ChapWorm/Audio/SC_ChapWorm_Glove_Energy_Maxed_OS
root=Modulator_0
nodes [
 {id=Modulator_0, children [WavePlayer_2]}
 {id=WavePlayer_2, properties {soundWaveAssetPtr=/ChapWormGameplay/Items/ChapWorm/Audio/wav/ChapWorm_Energy_Maxed_OS_01}}
]
properties {
 attenuationSettings=/ChapWormGameplay/Items/ChapWorm/Audio/Settings/CS_ChapWorm_Melee_Attack_Close_Range_Stereo_Att
 bOverrideConcurrency=true
 maxDistance=7000
 soundClassObject=&Fort_WeaponMelee_PSM
 volumeMultiplier=1
 concurrencyOverrides {maxCount=3}
}
