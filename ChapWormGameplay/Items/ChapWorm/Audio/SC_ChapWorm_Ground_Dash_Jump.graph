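# Sound cue for the ChapWorm item's ground dash jump, classed under player attacks and abilities.
# A mixer layers two random picks, one from four leap waves and one from three default leap waves.
# Each layer passes through its own modulator, with the four-wave layer given a wider volume variation.
# It borrows the close-range melee attack stereo attenuation and overrides concurrency to cap simultaneous plays and retriggering.
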SoundCue /ChapWormGameplay/Items/ChapWorm/Audio/SC_ChapWorm_Ground_Dash_Jump
root=Mixer_0
nodes [
 {id=Mixer_0, children [Modulator_1,Modulator_0], properties {inputVolume [1,1]}}
 {id=Modulator_0, children [Random_0], properties {volumeMax=1, volumeMin=0.6}}
 {id=Modulator_1, children [Random_1]}
 {id=Random_0, children [WavePlayer_0,WavePlayer_1,WavePlayer_2,WavePlayer_3], properties {weights [1,1,1,1]}}
 {id=Random_1, children [WavePlayer_4,WavePlayer_5,WavePlayer_6], properties {weights [1,1,1]}}
 {id=WavePlayer_0, properties {soundWaveAssetPtr=/ChapWormGameplay/Items/ChapWorm/Audio/wav/ChapWorm_Ground_Dash_Leap_01}}
 {id=WavePlayer_1, properties {soundWaveAssetPtr=/ChapWormGameplay/Items/ChapWorm/Audio/wav/ChapWorm_Ground_Dash_Leap_02}}
 {id=WavePlayer_2, properties {soundWaveAssetPtr=/ChapWormGameplay/Items/ChapWorm/Audio/wav/ChapWorm_Ground_Dash_Leap_03}}
 {id=WavePlayer_3, properties {soundWaveAssetPtr=/ChapWormGameplay/Items/ChapWorm/Audio/wav/ChapWorm_Ground_Dash_Leap_04}}
 {id=WavePlayer_4, properties {soundWaveAssetPtr=/ChapWormGameplay/Items/ChapWorm/Audio/wav/ChapWorm_Ground_Dash_Leap_Default_01}}
 {id=WavePlayer_5, properties {soundWaveAssetPtr=/ChapWormGameplay/Items/ChapWorm/Audio/wav/ChapWorm_Ground_Dash_Leap_Default_02}}
 {id=WavePlayer_6, properties {soundWaveAssetPtr=/ChapWormGameplay/Items/ChapWorm/Audio/wav/ChapWorm_Ground_Dash_Leap_Default_03}}
]
properties {
 attenuationSettings=/ChapWormGameplay/Items/ChapWorm/Audio/Settings/CS_ChapWorm_Melee_Attack_Close_Range_Stereo_Att
 bOverrideConcurrency=true
 maxDistance=7000
 soundClassObject=/Game/Sounds/FortSoundClassesAndModes/Fort_06_Player_Attacks_Abilities
 volumeMultiplier=0.8
 concurrencyOverrides {maxCount=3, retriggerTime=1.5}
}
