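# Sound cue for ChapWorm body impacts, played through the weapons-impact sound class with pickaxe player-impact attenuation.
# It mixes a random, modulated body-impact wave with a random flesh-impact layer that a fixed short delay offsets slightly behind it.
# A root modulator varies pitch and volume across the whole mix.
# A concurrency override limits simultaneous instances, stopping the oldest first.
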SoundCue /ChapWormGameplay/Items/ChapWorm/Audio/SC_ChapWorm_Impt_Body
root=Modulator_1
nodes [
 {id=Delay_0, children [Random_1], properties {delayMax=0.1, delayMin=0.1}}
 {id=Mixer_1, children [Modulator_0,Delay_0], properties {inputVolume [1,1]}}
 {id=Modulator_0, children [Random_0]}
 {id=Modulator_1, children [Mixer_1], properties {pitchMax=1.2, volumeMin=0.75}}
 {id=Random_0, children [WavePlayer_0,WavePlayer_1,WavePlayer_2,WavePlayer_3], properties {weights [1,1,1,1]}}
 {id=Random_1, children [WavePlayer_4,WavePlayer_5,WavePlayer_6,WavePlayer_7,WavePlayer_8], properties {weights [1,1,1,1,1]}}
 {id=WavePlayer_0, properties {soundWaveAssetPtr=/ChapWormGameplay/Items/ChapWorm/Audio/wav/ChapWorm_Impt_Body_01}}
 {id=WavePlayer_1, properties {soundWaveAssetPtr=/ChapWormGameplay/Items/ChapWorm/Audio/wav/ChapWorm_Impt_Body_02}}
 {id=WavePlayer_2, properties {soundWaveAssetPtr=/ChapWormGameplay/Items/ChapWorm/Audio/wav/ChapWorm_Impt_Body_03}}
 {id=WavePlayer_3, properties {soundWaveAssetPtr=/ChapWormGameplay/Items/ChapWorm/Audio/wav/ChapWorm_Impt_Body_04}}
 {id=WavePlayer_4, properties {soundWaveAssetPtr=/ChapWormGameplay/Items/ChapWorm/Audio/wav/ChapWorm_Impt_Flesh_01}}
 {id=WavePlayer_5, properties {soundWaveAssetPtr=/ChapWormGameplay/Items/ChapWorm/Audio/wav/ChapWorm_Impt_Flesh_02}}
 {id=WavePlayer_6, properties {soundWaveAssetPtr=/ChapWormGameplay/Items/ChapWorm/Audio/wav/ChapWorm_Impt_Flesh_03}}
 {id=WavePlayer_7, properties {soundWaveAssetPtr=/ChapWormGameplay/Items/ChapWorm/Audio/wav/ChapWorm_Impt_Flesh_04}}
 {id=WavePlayer_8, properties {soundWaveAssetPtr=/ChapWormGameplay/Items/ChapWorm/Audio/wav/ChapWorm_Impt_Flesh_05}}
]
properties {
 attenuationSettings=/Game/Sounds/Attenuation/Fort_Weapon_Pickaxe_Player_Impact_Att
 bOverrideConcurrency=true
 maxDistance=3512
 soundClassObject=/Game/Sounds/FortSoundClassesAndModes/Fort_Events_Weapons_Impact
 volumeMultiplier=0.7
 concurrencyOverrides {maxCount=3, resolutionRule=StopOldest, volumeScale=0.4}
}
