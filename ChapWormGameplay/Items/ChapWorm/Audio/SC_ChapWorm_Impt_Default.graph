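# Sound cue for the ChapWorm item's default impact, playing ground-impact sounds in the weapon impact sound class.
# A Random node picks one of four equally weighted ground-impact waves and feeds a Modulator at the root.
# It borrows the player pickaxe impact attenuation and halves the cue volume.
# A concurrency override caps simultaneous instances and applies volume scaling to overlapping plays.
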SoundCue /ChapWormGameplay/Items/ChapWorm/Audio/SC_ChapWorm_Impt_Default
root=Modulator_0
nodes [
 {id=Modulator_0, children [Random_1]}
 {id=Random_1, children [WavePlayer_4,WavePlayer_5,WavePlayer_6,WavePlayer_7], properties {weights [1,1,1,1]}}
 {id=WavePlayer_4, properties {soundWaveAssetPtr=/ChapWormGameplay/Items/ChapWorm/Audio/wav/ChapWorm_Impt_Ground_01}}
 {id=WavePlayer_5, properties {soundWaveAssetPtr=/ChapWormGameplay/Items/ChapWorm/Audio/wav/ChapWorm_Impt_Ground_02}}
 {id=WavePlayer_6, properties {soundWaveAssetPtr=/ChapWormGameplay/Items/ChapWorm/Audio/wav/ChapWorm_Impt_Ground_03}}
 {id=WavePlayer_7, properties {soundWaveAssetPtr=/ChapWormGameplay/Items/ChapWorm/Audio/wav/ChapWorm_Impt_Ground_04}}
]
properties {
 attenuationSettings=/Game/Sounds/Attenuation/Fort_Weapon_Pickaxe_Player_Impact_Att
 bOverrideConcurrency=true
 maxDistance=3512
 soundClassObject=/Game/Sounds/FortSoundClassesAndModes/Fort_Events_Weapons_Impact
 volumeMultiplier=0.5
 concurrencyOverrides {maxCount=3, volumeScale=0.4}
}
