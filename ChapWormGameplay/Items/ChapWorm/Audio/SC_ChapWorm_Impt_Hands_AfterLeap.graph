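# Sound cue for the ChapWorm item's hand impact after a leap.
# A Random node picks one of three impact wave variants with equal weight and feeds it through a Modulator for variation.
# It uses the player pickaxe impact attenuation and the weapon impact sound class.
# A concurrency override caps simultaneous instances, stopping the oldest and applying volume scaling.
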SoundCue /ChapWormGameplay/Items/ChapWorm/Audio/SC_ChapWorm_Impt_Hands_AfterLeap
root=Modulator_2
nodes [
 {id=Modulator_2, children [Random_2]}
 {id=Random_2, children [WavePlayer_10, WavePlayer_11, WavePlayer_12], properties {weights [1,1,1]}}
 {id=WavePlayer_10, properties {soundWaveAssetPtr=/ChapWormGameplay/Items/ChapWorm/Audio/wav/ChapWorm_Impt_Hands_AfterLeap_03}}
 {id=WavePlayer_11, properties {soundWaveAssetPtr=/ChapWormGameplay/Items/ChapWorm/Audio/wav/ChapWorm_Impt_Hands_AfterLeap_02}}
 {id=WavePlayer_12, properties {soundWaveAssetPtr=/ChapWormGameplay/Items/ChapWorm/Audio/wav/ChapWorm_Impt_Hands_AfterLeap_01}}
]
properties {
 attenuationSettings=/Game/Sounds/Attenuation/Fort_Weapon_Pickaxe_Player_Impact_Att
 bOverrideConcurrency=true
 maxDistance=3512
 soundClassObject=/Game/Sounds/FortSoundClassesAndModes/Fort_Events_Weapons_Impact
 concurrencyOverrides {maxCount=3, resolutionRule=StopOldest, volumeScale=0.4}
}
