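# Sound cue for the ChapWorm item's "Overpowered" impact, layering a surge sound over a ground hit.
# A mixer combines two modulated branches, each randomly picking one of four equally weighted waves, with the surge layer pitched up and the ground layer left at default modulation.
# It plays boosted in the impacts sound class with melee AOE slam attenuation and its own concurrency cap.
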SoundCue /ChapWormGameplay/Items/ChapWorm/Audio/SC_ChapWorm_Impt_Overpowered
root=Mixer_1
nodes [
 {id=Mixer_1, children [Modulator_0,Modulator_1], properties {inputVolume [1,1]}}
 {id=Modulator_0, children [Random_1], properties {pitchMax=1.3, pitchMin=1.2}}
 {id=Modulator_1, children [Random_0]}
 {id=Random_0, children [WavePlayer_5,WavePlayer_6,WavePlayer_7,WavePlayer_8], properties {weights [1,1,1,1]}}
 {id=Random_1, children [WavePlayer_0,WavePlayer_1,WavePlayer_2,WavePlayer_3], properties {weights [1,1,1,1]}}
 {id=WavePlayer_0, properties {soundWaveAssetPtr=/ChapWormGameplay/Items/ChapWorm/Audio/wav/ChapWorm_Impt_Surge_layer_01}}
 {id=WavePlayer_1, properties {soundWaveAssetPtr=/ChapWormGameplay/Items/ChapWorm/Audio/wav/ChapWorm_Impt_Surge_layer_02}}
 {id=WavePlayer_2, properties {soundWaveAssetPtr=/ChapWormGameplay/Items/ChapWorm/Audio/wav/ChapWorm_Impt_Surge_layer_03}}
 {id=WavePlayer_3, properties {soundWaveAssetPtr=/ChapWormGameplay/Items/ChapWorm/Audio/wav/ChapWorm_Impt_Surge_layer_04}}
 {id=WavePlayer_5, properties {soundWaveAssetPtr=/ChapWormGameplay/Items/ChapWorm/Audio/wav/ChapWorm_Impt_Ground_01}}
 {id=WavePlayer_6, properties {soundWaveAssetPtr=/ChapWormGameplay/Items/ChapWorm/Audio/wav/ChapWorm_Impt_Ground_02}}
 {id=WavePlayer_7, properties {soundWaveAssetPtr=/ChapWormGameplay/Items/ChapWorm/Audio/wav/ChapWorm_Impt_Ground_03}}
 {id=WavePlayer_8, properties {soundWaveAssetPtr=/ChapWormGameplay/Items/ChapWorm/Audio/wav/ChapWorm_Impt_Ground_04}}
]
properties {
 attenuationSettings=/Game/Sounds/Attenuation/Fort_Weapon_Melee_Attack_Special_AOE_Slam_Att
 bOverrideConcurrency=true
 maxDistance=13031.375
 soundClassObject=&Fort_Impacts
 volumeMultiplier=1.5
 concurrencyOverrides {maxCount=3}
}
