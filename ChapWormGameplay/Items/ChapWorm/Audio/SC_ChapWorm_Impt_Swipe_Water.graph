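# Sound cue for the ChapWorm swipe impact on water.
# A Random node picks among four equally weighted water impact waves and feeds a Modulator that adds pitch variation.
# It plays in the weapon impact sound class at reduced volume.
# Attenuation and concurrency are both overridden on the cue, with a capped voice count and volume scaling for overlapping hits.
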SoundCue /ChapWormGameplay/Items/ChapWorm/Audio/SC_ChapWorm_Impt_Swipe_Water
root=Modulator_0
nodes [
 {id=Modulator_0, children [Random_1], properties {pitchMax=1.1}}
 {id=Random_1, children [WavePlayer_2,WavePlayer_3,WavePlayer_4,WavePlayer_5], properties {weights [1,1,1,1]}}
 {id=WavePlayer_2, properties {soundWaveAssetPtr=/ChapWormGameplay/Items/ChapWorm/Audio/wav/ChapWorm_Impt_Water_01}}
 {id=WavePlayer_3, properties {soundWaveAssetPtr=/ChapWormGameplay/Items/ChapWorm/Audio/wav/ChapWorm_Impt_Water_02}}
 {id=WavePlayer_4, properties {soundWaveAssetPtr=/ChapWormGameplay/Items/ChapWorm/Audio/wav/ChapWorm_Impt_Water_03}}
 {id=WavePlayer_5, properties {soundWaveAssetPtr=/ChapWormGameplay/Items/ChapWorm/Audio/wav/ChapWorm_Impt_Water_04}}
]
properties {
 attenuationSettings=/Game/Sounds/Attenuation/Fort_Weapon_Pickaxe_Player_Impact_Att
 bOverrideAttenuation=true
 bOverrideConcurrency=true
 maxDistance=2456.6169
 soundClassObject=/Game/Sounds/FortSoundClassesAndModes/Fort_Events_Weapons_Impact
 volumeMultiplier=0.8
 attenuationOverrides {falloffDistance=2056.6169, stereoSpread=300}
 concurrencyOverrides {maxCount=6, volumeScale=0.4}
}
