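# Sound cue for the ChapWorm item's water splash impact, classed as a weapon impact sound.
# A Random node picks between two equally weighted splash waves and feeds a Modulator at the root.
# The cue overrides attenuation despite referencing a pickaxe impact preset, and its own concurrency limit stops the oldest instance when exceeded.
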SoundCue /ChapWormGameplay/Items/ChapWorm/Audio/SC_ChapWorm_Impt_WaterSplash
root=Modulator_0
nodes [
 {id=Modulator_0, children [Random_0]}
 {id=Random_0, children [WavePlayer_0,WavePlayer_1], properties {weights [1,1]}}
 {id=WavePlayer_0, properties {soundWaveAssetPtr=/ChapWormGameplay/Items/ChapWorm/Audio/wav/ChapWorm_Impt_Water_Splash_01}}
 {id=WavePlayer_1, properties {soundWaveAssetPtr=/ChapWormGameplay/Items/ChapWorm/Audio/wav/ChapWorm_Impt_Water_Splash_02}}
]
properties {
 attenuationSettings=/Game/Sounds/Attenuation/Fort_Weapon_Pickaxe_Player_Impact_Att
 bOverrideAttenuation=true
 bOverrideConcurrency=true
 maxDistance=2456.6169
 soundClassObject=/Game/Sounds/FortSoundClassesAndModes/Fort_Events_Weapons_Impact
 volumeMultiplier=1
 attenuationOverrides {falloffDistance=2056.6169, stereoSpread=300}
 concurrencyOverrides {maxCount=6, resolutionRule=StopOldest, volumeScale=0.4}
}
