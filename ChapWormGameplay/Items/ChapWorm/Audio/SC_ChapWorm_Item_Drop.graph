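# Sound cue for the ChapWorm item drop, played as a one-shot pickup-style world sound.
# A single wave player feeds the drop sample through a Modulator node at default settings into the output.
# It uses the shared weapon pickup attenuation and melee weapon sound class, with its own concurrency override capping simultaneous instances.
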SoundCue /ChapWormGameplay/Items/ChapWorm/Audio/SC_ChapWorm_Item_Drop
root=Modulator_1
nodes [
 {id=Modulator_1, children [WavePlayer_5]}
 {id=WavePlayer_5, properties {soundWaveAssetPtr=/ChapWormGameplay/Items/ChapWorm/Audio/wav/ChapWorm_Item_Drop_01}}
]
properties {
 attenuationSettings=&Fort_Weapon_Pickup_Att
 bOverrideConcurrency=true
 maxDistance=3500
 soundClassObject=&Fort_WeaponMelee_PSM
 volumeMultiplier=1
 concurrencyOverrides {maxCount=3}
}
