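# Sound Cue for the ChapWorm item's reload-end equip sound, named as the zero-energy variant.
# It picks one of two ReloadEnd waves at equal weight and passes it through a Modulator left at default settings.
# It uses the weapon pickup attenuation and melee weapon sound class, with a concurrency override that caps simultaneous instances.
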SoundCue /ChapWormGameplay/Items/ChapWorm/Audio/SC_ChapWorm_Item_Equip_ReloadEnd_ZeroEnergy
root=Modulator_2
nodes [
 {id=Modulator_2, children [Random_0]}
 {id=Random_0, children [WavePlayer_1,WavePlayer_2], properties {weights [1,1]}}
 {id=WavePlayer_1, properties {soundWaveAssetPtr=/ChapWormGameplay/Items/ChapWorm/Audio/wav/ChapWorm_Item_Equip_ReloadEnd_02}}
 {id=WavePlayer_2, properties {soundWaveAssetPtr=/ChapWormGameplay/Items/ChapWorm/Audio/wav/ChapWorm_Item_Equip_ReloadEnd_01}}
]
properties {
 attenuationSettings=&Fort_Weapon_Pickup_Att
 bOverrideConcurrency=true
 maxDistance=3500
 soundClassObject=&Fort_WeaponMelee
 volumeMultiplier=1
 concurrencyOverrides {maxCount=3}
}
