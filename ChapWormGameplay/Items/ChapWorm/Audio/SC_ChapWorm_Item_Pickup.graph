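# Sound cue played for picking up the ChapWorm item.
# A random node picks one of two equally weighted pickup wave variations and feeds it through a modulator at the root.
# It uses the weapon pickup attenuation and the melee weapon sound class.
# A concurrency override on the cue caps how many instances play at once.
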SoundCue /ChapWormGameplay/Items/ChapWorm/Audio/SC_ChapWorm_Item_Pickup
root=Modulator_0
nodes [
 {id=Modulator_0, children [Random_1]}
 {id=Random_1, children [WavePlayer_5,WavePlayer_6], properties {weights [1,1]}}
 {id=WavePlayer_5, properties {soundWaveAssetPtr=/ChapWormGameplay/Items/ChapWorm/Audio/wav/ChapWorm_Item_Pickup_01}}
 {id=WavePlayer_6, properties {soundWaveAssetPtr=/ChapWormGameplay/Items/ChapWorm/Audio/wav/ChapWorm_Item_Pickup_02}}
]
properties {
 attenuationSettings=&Fort_Weapon_Pickup_Att
 bOverrideConcurrency=true
 maxDistance=3500
 soundClassObject=&Fort_WeaponMelee_PSM
 volumeMultiplier=1
 concurrencyOverrides {maxCount=3}
}
