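# Sound cue for the ChapWorm item's OverPower activation, playing the overpowered loop-start wave.
# A single wave player feeds a Modulator root left at default settings, with the cue's overall volume boosted.
# It uses the close-range stereo melee attack attenuation and the melee weapon sound class, and overrides concurrency to cap simultaneous instances.
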SoundCue /ChapWormGameplay/Items/ChapWorm/Audio/SC_ChapWorm_OverPower_Activate
root=Modulator_0
nodes [
 {id=Modulator_0, children [WavePlayer_5]}
 {id=WavePlayer_5, properties {soundWaveAssetPtr=/ChapWormGameplay/Items/ChapWorm/Audio/wav/ChapWorm_Overpowered_lp_Start_01}}
]
properties {
 attenuationSettings=/ChapWormGameplay/Items/ChapWorm/Audio/Settings/CS_ChapWorm_Melee_Attack_Close_Range_Stereo_Att
 bOverrideConcurrency=true
 maxDistance=7000
 soundClassObject=&Fort_WeaponMelee_PSM
 volumeMultiplier=2
 concurrencyOverrides {maxCount=3}
}
