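# Sound cue for the ChapWorm item's OverPower cast, reusing an equip/reload-end wave recording.
# A single wave player feeds a Modulator that applies a fixed volume boost, with no random variation in volume.
# It plays through the melee weapon sound class with a close-range stereo melee attack attenuation asset, and overrides concurrency to cap simultaneous instances.
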SoundCue /ChapWormGameplay/Items/ChapWorm/Audio/SC_ChapWorm_OverPower_Cast
root=Modulator_1
nodes [
 {id=Modulator_1, children [WavePlayer_5], properties {volumeMax=4, volumeMin=4}}
 {id=WavePlayer_5, properties {soundWaveAssetPtr=/ChapWormGameplay/Items/ChapWorm/Audio/wav/SC_ChapWorm_Item_Equip_ReloadEnd_02}}
]
properties {
 attenuationSettings=/ChapWormGameplay/Items/ChapWorm/Audio/Settings/CS_ChapWorm_Melee_Attack_Close_Range_Stereo_Att
 bOverrideConcurrency=true
 maxDistance=7000
 soundClassObject=&Fort_WeaponMelee_PSM
 volumeMultiplier=1
 concurrencyOverrides {maxCount=3}
}
