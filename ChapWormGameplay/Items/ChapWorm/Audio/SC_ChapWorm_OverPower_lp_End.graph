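# Sound cue for the ChapWorm item that plays the ending sound of its overpowered loop.
# A Random node picks one of two end-variation waves with equal weight and feeds a Modulator at the root.
# It uses the ChapWorm close-range stereo melee attenuation and the weapon melee sound class, with overridden concurrency capping simultaneous plays.
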SoundCue /ChapWormGameplay/Items/ChapWorm/Audio/SC_ChapWorm_OverPower_lp_End
root=Modulator_0
nodes [
 {id=Modulator_0, children [Random_0]}
 {id=Random_0, children [WavePlayer_1,WavePlayer_2], properties {weights [1,1]}}
 {id=WavePlayer_1, properties {soundWaveAssetPtr=/ChapWormGameplay/Items/ChapWorm/Audio/wav/ChapWorm_Overpowered_lp_End_02}}
 {id=WavePlayer_2, properties {soundWaveAssetPtr=/ChapWormGameplay/Items/ChapWorm/Audio/wav/ChapWorm_Overpowered_lp_End_01}}
]
properties {
 attenuationSettings=/ChapWormGameplay/Items/ChapWorm/Audio/Settings/CS_ChapWorm_Melee_Attack_Close_Range_Stereo_Att
 bOverrideConcurrency=true
 maxDistance=7000
 soundClassObject=&Fort_WeaponMelee_PSM
 volumeMultiplier=0.7
 concurrencyOverrides {maxCount=3}
}
