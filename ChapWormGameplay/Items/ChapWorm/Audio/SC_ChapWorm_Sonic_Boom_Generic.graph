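# Sound cue for the ChapWorm item's generic sonic boom, classed as melee weapon audio.
# A random node picks one of four equally weighted boom waves, then feeds it through a modulator at the root.
# The cue boosts volume, uses the ChapWorm close-range stereo melee attack attenuation, and overrides concurrency to cap simultaneous instances at three.
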SoundCue /ChapWormGameplay/Items/ChapWorm/Audio/SC_ChapWorm_Sonic_Boom_Generic
root=Modulator_0
nodes [
 {id=Modulator_0, children [Random_2]}
 {id=Random_2, children [WavePlayer_12, WavePlayer_13, WavePlayer_14, WavePlayer_15], properties {weights [1,1,1,1]}}
 {id=WavePlayer_12, properties {soundWaveAssetPtr=/ChapWormGameplay/Items/ChapWorm/Audio/wav/ChapWorm_Sonic_Boom_01}}
 {id=WavePlayer_13, properties {soundWaveAssetPtr=/ChapWormGameplay/Items/ChapWorm/Audio/wav/ChapWorm_Sonic_Boom_02}}
 {id=WavePlayer_14, properties {soundWaveAssetPtr=/ChapWormGameplay/Items/ChapWorm/Audio/wav/ChapWorm_Sonic_Boom_03}}
 {id=WavePlayer_15, properties {soundWaveAssetPtr=/ChapWormGameplay/Items/ChapWorm/Audio/wav/ChapWorm_Sonic_Boom_04}}
]
properties {
 attenuationSettings=/ChapWormGameplay/Items/ChapWorm/Audio/Settings/CS_ChapWorm_Melee_Attack_Close_Range_Stereo_Att
 bOverrideConcurrency=true
 maxDistance=7000
 soundClassObject=&Fort_WeaponMelee_PSM
 volumeMultiplier=1.4
 concurrencyOverrides {maxCount=3}
}
