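# Sound cue for the ChapWorm item's overpowered sonic boom sound.
# A random node picks one of two equally weighted wave variations and feeds it through a modulator at the root.
# It plays in the melee weapon sound class with close-range stereo melee attenuation, a boosted volume and a concurrency override capping simultaneous instances.
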SoundCue /ChapWormGameplay/Items/ChapWorm/Audio/SC_ChapWorm_Sonic_Boom_Overpowered
root=Modulator_0
nodes [
 {id=Modulator_0, children [Random_1]}
 {id=Random_1, children [WavePlayer_0,WavePlayer_1], properties {weights [1,1]}}
 {id=WavePlayer_0, properties {soundWaveAssetPtr=/ChapWormGameplay/Items/ChapWorm/Audio/wav/ChapWorm_Sonic_Boom_Overpowered_01}}
 {id=WavePlayer_1, properties {soundWaveAssetPtr=/ChapWormGameplay/Items/ChapWorm/Audio/wav/ChapWorm_Sonic_Boom_Overpowered_02}}
]
properties {
 attenuationSettings=/ChapWormGameplay/Items/ChapWorm/Audio/Settings/CS_ChapWorm_Melee_Attack_Close_Range_Stereo_Att
 bOverrideConcurrency=true
 maxDistance=7000
 soundClassObject=&Fort_WeaponMelee_PSM
 volumeMultiplier=2
 concurrencyOverrides {maxCount=3}
}
