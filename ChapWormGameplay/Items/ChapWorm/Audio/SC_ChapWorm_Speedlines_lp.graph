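# Looping speedlines sound cue for the ChapWorm item, classed as a melee weapon sound.
# A looping wave player feeds an enveloper that swells volume from silence to full over about two and a half seconds.
# A modulator holds volume steady, then a root oscillator modulates pitch around a randomised centre at or slightly above normal.
# It uses close-range melee attack attenuation and overrides concurrency to cap simultaneous instances.
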SoundCue /ChapWormGameplay/Items/ChapWorm/Audio/SC_ChapWorm_Speedlines_lp
root=Oscillator_0
nodes [
 {
  id=Enveloper_1
  children [WavePlayer_4]
  properties {
   pitchCurve {editorCurveData {keys [{time=0}]}}
   volumeCurve {editorCurveData {keys [
      <at 0 0>
      {arriveTangent=0.1128, interpMode=RCIM_Cubic, leaveTangent=0.1128, tangentMode=RCTM_User, time=0.0737, value=0.0022}
      {arriveTangent=1.3251, interpMode=RCIM_Cubic, leaveTangent=1.3251, tangentMode=RCTM_User, time=2.5, value=1}
     ]}}
  }
 }
 {id=Modulator_0, children [Enveloper_1], properties {volumeMax=1, volumeMin=1}}
 {
  id=Oscillator_0
  children [Modulator_0]
  properties {bModulatePitch=true, centerMax=1.2, centerMin=1, frequencyMax=1.1, frequencyMin=0.9, offsetMax=0.8, offsetMin=1}
 }
 {id=WavePlayer_4, properties {bLooping=true, soundWaveAssetPtr=/ChapWormGameplay/Items/ChapWorm/Audio/wav/ChapWorm_Speedlines_lp}}
]
properties {
 attenuationSettings=/ChapWormGameplay/Items/ChapWorm/Audio/Settings/CS_ChapWorm_Melee_Attack_Close_Range_Stereo_Att
 bOverrideConcurrency=true
 maxDistance=7000
 soundClassObject=&Fort_WeaponMelee
 volumeMultiplier=0.32
 concurrencyOverrides {maxCount=3}
}
