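# Sound cue for a heavy ChapWorm melee swipe whoosh, routed to the melee weapon sound class.
# An equal-weight Random node picks one of five light attack whoosh waves and feeds a Modulator that randomizes volume and lowers pitch.
# The heavy variant thus reuses the light recordings pitched down rather than separate samples.
# Attenuation and concurrency are overridden on the cue, with concurrency set to prevent new instances once the limit is reached.
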SoundCue /ChapWormGameplay/Items/ChapWorm/Audio/SC_ChapWorm_Swipe_Simple_Heavy
root=Modulator_2
nodes [
 {id=Modulator_2, children [Random_1], properties {pitchMax=0.9, pitchMin=0.7, volumeMax=1, volumeMin=0.7}}
 {id=Random_1, children [WavePlayer_5,WavePlayer_6,WavePlayer_7,WavePlayer_8,WavePlayer_9], properties {weights [1,1,1,1,1]}}
 {id=WavePlayer_5, properties {soundWaveAssetPtr=/ChapWormGameplay/Items/ChapWorm/Audio/wav/ChapWorm_Whoosh_Attk_Light_01}}
 {id=WavePlayer_6, properties {soundWaveAssetPtr=/ChapWormGameplay/Items/ChapWorm/Audio/wav/ChapWorm_Whoosh_Attk_Light_02}}
 {id=WavePlayer_7, properties {soundWaveAssetPtr=/ChapWormGameplay/Items/ChapWorm/Audio/wav/ChapWorm_Whoosh_Attk_Light_03}}
 {id=WavePlayer_8, properties {soundWaveAssetPtr=/ChapWormGameplay/Items/ChapWorm/Audio/wav/ChapWorm_Whoosh_Attk_Light_04}}
 {id=WavePlayer_9, properties {soundWaveAssetPtr=/ChapWormGameplay/Items/ChapWorm/Audio/wav/ChapWorm_Whoosh_Attk_Light_05}}
]
properties {
 bOverrideAttenuation=true
 bOverrideConcurrency=true
 maxDistance=4000
 soundClassObject=&Fort_WeaponMelee
 volumeMultiplier=0.6
 concurrencyOverrides {resolutionRule=PreventNew}
}
