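# Sound cue for the surged variant of the ChapWorm melee weapon's third swipe attack.
# A Mixer layers a random pick of five swipe waves, boosted in volume, with a random pick of five surge-layer waves that get randomized volume and raised pitch.
# Playback uses close-range melee attenuation, the melee weapon sound class and a surged-melee concurrency set.
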SoundCue /ChapWormGameplay/Items/ChapWorm/Audio/SC_ChapWorm_Swipe_Surge_03
root=Mixer_0
nodes [
 {id=Mixer_0, children [Modulator_0,Modulator_2], properties {inputVolume [1,1]}}
 {id=Modulator_0, children [Random_0], properties {volumeMax=2, volumeMin=2}}
 {id=Modulator_2, children [Random_1], properties {pitchMax=1.3, pitchMin=1, volumeMin=0.6}}
 {id=Random_0, children [WavePlayer_0,WavePlayer_1,WavePlayer_2,WavePlayer_3,WavePlayer_4], properties {weights [1,1,1,1,1]}}
 {id=Random_1, children [WavePlayer_5,WavePlayer_6,WavePlayer_7,WavePlayer_8,WavePlayer_9], properties {weights [1,1,1,1,1]}}
 {id=WavePlayer_0, properties {soundWaveAssetPtr=/ChapWormGameplay/Items/ChapWorm/Audio/wav/ChapWorm_Attk_Swipe_03_01}}
 {id=WavePlayer_1, properties {soundWaveAssetPtr=/ChapWormGameplay/Items/ChapWorm/Audio/wav/ChapWorm_Attk_Swipe_03_02}}
 {id=WavePlayer_2, properties {soundWaveAssetPtr=/ChapWormGameplay/Items/ChapWorm/Audio/wav/ChapWorm_Attk_Swipe_03_03}}
 {id=WavePlayer_3, properties {soundWaveAssetPtr=/ChapWormGameplay/Items/ChapWorm/Audio/wav/ChapWorm_Attk_Swipe_03_04}}
 {id=WavePlayer_4, properties {soundWaveAssetPtr=/ChapWormGameplay/Items/ChapWorm/Audio/wav/ChapWorm_Attk_Swipe_03_05}}
 {id=WavePlayer_5, properties {soundWaveAssetPtr=/ChapWormGameplay/Items/ChapWorm/Audio/wav/ChapWorm_Attk_Swipe_SurgeLayer_01}}
 {id=WavePlayer_6, properties {soundWaveAssetPtr=/ChapWormGameplay/Items/ChapWorm/Audio/wav/ChapWorm_Attk_Swipe_SurgeLayer_02}}
 {id=WavePlayer_7, properties {soundWaveAssetPtr=/ChapWormGameplay/Items/ChapWorm/Audio/wav/ChapWorm_Attk_Swipe_SurgeLayer_03}}
 {id=WavePlayer_8, properties {soundWaveAssetPtr=/ChapWormGameplay/Items/ChapWorm/Audio/wav/ChapWorm_Attk_Swipe_SurgeLayer_04}}
 {id=WavePlayer_9, properties {soundWaveAssetPtr=/ChapWormGameplay/Items/ChapWorm/Audio/wav/ChapWorm_Attk_Swipe_SurgeLayer_05}}
]
properties {
 attenuationSettings=/ChapWormGameplay/Items/ChapWorm/Audio/Settings/CS_ChapWorm_Melee_Attack_Close_Range_Stereo_Att
 maxDistance=7000
 soundClassObject=&Fort_WeaponMelee_PSM
 volumeMultiplier=1.2
 concurrencyOverrides {maxCount=3, resolutionRule=StopOldest, volumeScale=0.2}
 concurrencySet [/ChapWormGameplay/Items/ChapWorm/Audio/Settings/CS_ChapWorm_Melee_Surged]
}
