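; Character part modifier for the ChapWorm gauntlet that drives the part's material from the weapon's remaining ammo.
; After res-in, it binds to the current ChapWorm weapon's local ammo change event, applies the initial count and stamps game time into custom primitive data on the part's skeletal mesh.
; Ammo updates write a converted scalar to another custom data slot on that mesh.
; Pre-destruction unbinds the delegate; res-in and update failures are logged.
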(event EventOnPostResIn
  (bind _returnvalue (CustomCharacter|GetSkeletalMeshForPartOnPlayerPawn))
  (bind _returnvalue_2 (CustomCharacter|GetAssociatedPlayerPawn))
  (bind _returnvalue_3 (Weapon|GetCurrentWeapon _returnvalue_2))
  ;[0.07224 0.2865 0.1049 0.5] OnPostResIn: Bind to know when the ammo changes to update the material.
  (|CallParentFunction)
  (Utilities|IsValid _returnvalue_2
    (:Is%20Valid
      (Utilities|IsValid _returnvalue_3
        (:Is%20Valid
          (Utilities|Casting|CastToB_Weapon_ChapWorm _returnvalue_3
            (:then
              (EventDispatchers|BindEventtoOnLocalAmmoChanged _returnvalue_3 (EventDispatchers|CreateEvent :!event Custom|UpdatingAmmoCount))
              (CallFunction|UpdatingAmmoCount 0 (FortWeapon|GetRemainingAmmo _returnvalue_3))
              (Utilities|IsValid _returnvalue
                (:Is%20Valid
                  (Rendering|Material|SetCustomPrimitiveDataFloat _returnvalue 11 (Utilities|Time|GetGameTimeinSeconds)))
                (:Is%20Not%20Valid
                  (Fortnite|FortShippingLog self "OnPostResIn: SkeletalMesh is null." true))))
            (:CastFailed
              (Fortnite|FortShippingLog self (Utilities|String|Append "OnPostResIn: CurrentWeapon is NOT ChapWorm: ." (Utilities|GetDisplayName _returnvalue_3)) true))))
        (:Is%20Not%20Valid
          (Fortnite|FortShippingLog self "OnPostResIn: CurrentWeapon is null." true))))
    (:Is%20Not%20Valid
      (Fortnite|FortShippingLog self "OnPostResIn: Player pawn is null." true))))

(event Custom|UpdatingAmmoCount
  (params (LocalCount :type Integer) (LocalRemaining :type Integer))
  (bind _returnvalue (CustomCharacter|GetSkeletalMeshForPartOnPlayerPawn))
  ;[0.1379 0.02737 0.15 0.5] UpdatingAmmoCount: Adjust the material based on the latest ammo count.
  (Utilities|IsValid _returnvalue
    (:Is%20Valid
      (Rendering|Material|SetCustomPrimitiveDataFloat _returnvalue 10 (Utilities|AmmotoMatScalar LocalRemaining)))
    (:Is%20Not%20Valid
      (Fortnite|FortShippingLog self "UpdatingAmmoCount: SkeletalMesh is null." true))))

(event EventPreDestruction
  (bind _returnvalue (CustomCharacter|GetAssociatedPlayerPawn))
  (bind _returnvalue_2 (Weapon|GetCurrentWeapon _returnvalue))
  ;[0.1024 0.06429 0.01266 0.5] PreDestruction: Unbind delegates.
  (|CallParentFunction)
  (Utilities|IsValid _returnvalue
    (:Is%20Valid
      (Utilities|IsValid _returnvalue_2
        (:Is%20Valid
          (EventDispatchers|UnbindEventfromOnLocalAmmoChanged _returnvalue_2 (EventDispatchers|CreateEvent :!event Custom|UpdatingAmmoCount)))))))

(fn ConstructionScript)
