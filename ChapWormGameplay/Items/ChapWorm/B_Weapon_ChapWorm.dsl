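; ChapWorm weapon Blueprint that plays a cosmetic equip effect and replaces the out-of-ammo text with "SUPERCHARGED".
; On equip and attach, it async-loads the default or Overpower Niagara system, chosen by the owner pawn's gameplay tag, and attaches it to the mesh.
; Overpowered pawns also spawn a radial-force field actor, skipped on mobile, PC performance mode or low detail.
; An authority-only debug event resets ammo to a given count.
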(event FortWeapon|EventOnEquippedAndAttached
  ;[0.05089 0.008499 0.08333] OnEquippedAndAttached: Play the equip fx when the weapon is equipped and attached.
  (|CallParentFunction)
  (Networking|CanExecuteCosmeticEvents
    (:True
      (bind _asfort_player_pawn (Utilities|Casting|CastToFortPlayerPawn (Actor|GetOwner self)))
      (bind _mesh (Variables|/Script/Engine.Character|GetMesh _asfort_player_pawn)
        (:then
          (bind _object (Utilities|AsyncLoadAsset (select (GameplayTags|HasMatchingGameplayTag _asfort_player_pawn "(TagName=\"ChapWorm.Overpower\")") (Variables|Settings|GetEquipOverpowerNiagaraSystem) (Variables|Settings|GetEquipDefaultNiagaraSystem)))
            (:Completed
              (Niagara|SpawnSystemAttached (Utilities|Casting|CastToNiagaraSystem _object) _mesh "None" "0, 0, 0" "0, 0, 0" "KeepRelativeOffset" true true "AutoRelease")
              (if (GameplayTags|HasMatchingGameplayTag _asfort_player_pawn "(TagName=\"ChapWorm.Overpower\")")
                ;[1 0 0] VFX: Radial Force
                (if (Math|Boolean|ANDBoolean (not (Fortnite|IsMobilePlatform)) (not (Class|PlatformHelpersLibrary|IsPcPerfMode)) (> (Rendering|GetRenderingDetailMode) 1))
                  (Game|SpawnActorfromClass "/Game/Effects/Niagara/Blueprints/PhsyicsFields/FN_RadialForce.FN_RadialForce_C" (Math|Transform|MakeTransform (Transformation|GetWorldLocation (Variables|/Script/Engine.Actor|GetRootComponent _asfort_player_pawn))) "Undefined" "MultiplyWithRoot" 0 3.0 0.6 1e+03 0.9 0.214 0))))))
        (:else)))))

(event Custom|DebugBlackPantherAmmo
  (params (AmmoCount :type Integer))
  ;[0 0 0] DebugBlackPantherAmmo: Used to change the ammo of the black panther weapon.
  (Networking|SwitchHasAuthority
    (:Authority
      (FortWeapon|ModAmmo (* (FortWeapon|GetMagazineAmmoCount) -1))
      (FortWeapon|ModAmmo AmmoCount))))

(fn GetOutOfAmmoTextOverride
  (return "NSLOCTEXT(\"[C591F885B81F78CD46E2C6AEE427A183]\", \"7E59C17946B89DA432CD45A16BB593C2\", \"SUPERCHARGED\")"))

(fn ConstructionScript
  (|CallParentFunction))
