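# Unlit translucent material for the ChapWorm claw trail effect, built for Niagara ribbons and mesh particles.
# Emissive blends a color gradient with a rotating camera-vector galaxy texture and distorted panning highlight, tinted by particle color.
# Opacity shapes ridge and thickness procedurally from UV1 sine curves, with edge fade, particle alpha and non-mobile depth fade.
# A separate branch multiplies that color by two distorted noise layers, offset by the Random dynamic parameter.
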Material /ChapWormGameplay/Items/ChapWorm/FX/Materials/M_ChapWorm_Claw_Trail
graph {
 properties {
  bUsedWithNiagaraMeshParticles=true
  bUsedWithNiagaraRibbons=true
  blendMode=BLEND_Translucent
  pixelDepthOffsetMode=PDOM_Legacy
  shadingModel=MSM_Unlit
  usedShadingModels=Unlit
 }
 nodes [
  n1 = TextureCoordinate()
  n2 = VectorParameter() {group=HighLight, parameterName=HighLight UV, sortPriority=1, defaultValue=<rgba 2 1 2 0>}
  n3 = AppendVector(A=@n2.R, B=@n2.R)
  n4 = TextureCoordinate() {uTiling=0.7, vTiling=0.7}
  n5 = VectorParameter() {group=Distortion, parameterName=Distortion UV, sortPriority=1, defaultValue=<rgba 1 1 0 0>}
  n6 = AppendVector(A=@n5.R, B=@n5.R)
  n7 = TextureSampleParameter2D(UVs=Panner(Coordinate=Multiply(A=@n4, B=@n6), Speed=AppendVector(A=@n5.B, B=@n5.B))) {
   group=Distortion
   parameterName="Distortion  Tex"
   samplerType=SAMPLERTYPE_Normal
   sortPriority=0
   texture=/Game/Effects/Fort_Effects/Emotes/BirdNestBlue/T_SwirlNoise
  }
  n8 = ComponentMask(None=@n4) {g=true}
  n9 = Multiply(
   A=ComponentMask(None=@n7.RGB) {g=true, r=true}
   B=LinearInterpolate(B=OneMinus(None=@n8), Alpha=ScalarParameter() {defaultValue=1, group=Distortion, parameterName=Distortion Gradient Lerp, sortPriority=3}) {constA=1, constAlpha=0}
  )
  n10 = Add(
   A=Multiply(A=@n9, B=ScalarParameter() {defaultValue=0.05, group=Distortion, parameterName=Distortion Intensity HighLight, sortPriority=2})
   B=Panner(Coordinate=Multiply(A=@n1, B=@n3), Speed=AppendVector(A=@n2.B, B=@n2.B))
  )
  n11 = VectorParameter() {group=HighLight, parameterName=HighLight Refine, channelNames {a=Lerp, b=Base, g=Mult, r=Power}, defaultValue=<rgba 0.5 30 1 0.5>}
  n12 = Multiply(
   A=MaterialFunctionCall(
    Texture=TextureSample(UVs=@n10) {samplerSource=SSM_Wrap_WorldGroupSettings, texture=/Game/Weapons/FORT_Melee/Pickaxe_Inferno/FX/T_Pickaxe_Inferno_Trail_01}.R
    RGB=@n11
    A=@n11
   ) {materialFunction=&MF_FX_Refine}.Result
   B=VectorParameter() {group=HighLight, parameterName=HighLight Color, defaultValue=<rgba 0.6667 0 1 1>}.RGB
  )
  n13 = TextureCoordinate()
  n14 = LinearInterpolate(
   B=OneMinus(None=ComponentMask(None=@n13) {g=true})
   Alpha=ScalarParameter() {group=Distortion, parameterName=Distortion Gradient Lerp, sortPriority=3}
  ) {constA=1, constAlpha=0}
  n15 = TextureCoordinate()
  n16 = Power(Base=OneMinus(None=ComponentMask(None=@n15) {r=true}), Exp=ScalarParameter() {defaultValue=8, group=Color, parameterName=Color Power}) {constExponent=4}
  n17 = Multiply(
   A=LinearInterpolate(
    A=VectorParameter() {group=Color, parameterName=Color A, defaultValue=<rgba 1.1146 0.125 2.5 1>}.RGB
    B=VectorParameter() {group=Color, parameterName=Color B, defaultValue=<rgba 0 13.3333 40 1>}.RGB
    Alpha=Saturate(None=@n16)
   )
   B=ScalarParameter() {defaultValue=1, group=Color, parameterName=Color Mult}
  )
  n18 = VectorParameter() {group=Galaxy, parameterName=Galaxy Color, defaultValue=<rgba 1 1 1 1>}
  n19 = TextureObjectParameter() {
   group=Galaxy
   parameterName=GalaxyTexture
   samplerType=SAMPLERTYPE_LinearGrayscale
   texture=/Game/Effects/Fort_Effects/Emotes/UltraLight/Texture/T_PlasmaCoils
  }
  n20 = Time()
  n21 = Multiply(A=@n20, B=Constant())
  n22 = ScalarParameter() {group=Galaxy, parameterName=Background Rotation}
  n23 = AppendVector(
   A=ScalarParameter() {defaultValue=0.02, group=Galaxy, parameterName=Galaxy Rotation Speed}
   B=ScalarParameter() {defaultValue=-0.02, group=Galaxy, parameterName=Galaxy Rotation Speed_1}
  )
  n24 = ScalarParameter() {defaultValue=0.1, group=Galaxy, parameterName=Galaxy Rotation Speed_2}
  n25 = MaterialFunctionCall!(
   CameraVector
   Rotation Speed=Multiply(A=LinearInterpolate(A=@n20, B=@n21, Alpha=@n22), B=AppendVector(A=@n23, B=@n24))
   Texture=@n19
   Tiling=ScalarParameter() {defaultValue=1, group=Galaxy, parameterName=GalaxyTiling}
   Offset=Constant()
   AxisFadeContrast=ScalarParameter() {defaultValue=6, group=Galaxy, parameterName=GalaxyAxisFade}
  ) {materialFunction=/Game/Weapons/WeaponSkins/MaterialFunction/TriplanarCameraVector_Rotated}
  n26 = LinearInterpolate(
   A=@n17
   B=Multiply(A=Multiply(A=@n18.RGB, B=@n25.Result), B=@n17)
   Alpha=ScalarParameter() {defaultValue=0.5, group=Galaxy, parameterName=Lerp Galaxy, sortPriority=0}
  )
  n27 = LinearInterpolate(
   A=@n26
   B=Multiply(A=Multiply(A=@n12, B=@n14), B=@n26)
   Alpha=ScalarParameter() {defaultValue=0.5, group=HighLight, parameterName=Lerp Highlight}
  )
  n28 = ParticleColor()
  MaterialFunctionCall(Emissive=Multiply(A=@n27, B=@n28.RGB)) {materialFunction=&MF_FX_EmissiveAdjustments}
  n29 = MaterialFunctionCall(Float2=TextureCoordinate() {coordinateIndex=1}) {materialFunction=&BreakOutFloat2Components}
  n30 = Sine(None=Multiply(A=@n29.G) {constB=0.5})
  n31 = LinearInterpolate(
   A=ScalarParameter() {defaultValue=1, group=Shape, parameterName=Ridge Top, sortPriority=4}
   B=ScalarParameter() {defaultValue=64, group=Shape, parameterName=Ridge Bottom, sortPriority=5}
   Alpha=@n29.R
  ) {constA=6, constB=2}
  n32 = Multiply(A=Saturate(None=Power(Base=@n30, Exp=@n31) {constExponent=4}), B=ScalarParameter() {group=Shape, parameterName=Ridge Mult, sortPriority=6}) {constB=0}
  n33 = OneMinus(None=@n29.R)
  n34 = Power(
   Base=Sine(None=Multiply(A=@n29.G) {constB=0.5})
   Exp=LinearInterpolate(
    A=ScalarParameter() {defaultValue=0.5, group=Shape, parameterName=Thickness Top, sortPriority=0}
    B=ScalarParameter() {group=Shape, parameterName=Thickness Bottom, sortPriority=1}
    Alpha=@n33
   ) {constA=6, constB=2}
  ) {constExponent=4}
  n35 = Multiply(
   A=Power(Base=@n34, Exp=ScalarParameter() {defaultValue=1, group=Shape, parameterName=Base Power})
   B=ScalarParameter() {defaultValue=1, group=Shape, parameterName=Base Mult}
  ) {constB=0}
  n36 = Multiply(A=Saturate(None=Add(A=@n32, B=@n35)), B=Multiply(A=@n34, B=@n33))
  n37 = LinearInterpolate(
   A=ScalarParameter() {defaultValue=1, group=Shape, parameterName=Fade Top, sortPriority=2}
   B=ScalarParameter() {defaultValue=1, group=Shape, parameterName=Fade Bottom, sortPriority=3}
   Alpha=@n33
  ) {constA=1, constB=0.5}
  n38 = TextureCoordinate()
  n39 = ComponentMask(None=@n38) {g=true}
  n40 = Multiply(A=@n39, B=OneMinus(None=@n39))
  n41 = ScalarParameter() {defaultValue=4, group=Shape, parameterName=Edge Fade Power}
  n42 = Power(Base=ComponentMask(None=@n38) {r=true}) {constExponent=3}
  n43 = Multiply(A=Saturate(None=Multiply(A=@n40, B=@n41)), B=Saturate(None=Multiply(A=@n42) {constB=20}))
  n44 = ScalarParameter() {defaultValue=1, group=Shape, parameterName=Use Edge Fade}
  n45 = Multiply(
   A=Multiply(A=Power(Base=@n36, Exp=@n37), B=LinearInterpolate(B=@n43, Alpha=@n44) {constA=1})
   B=ScalarParameter() {defaultValue=6, group=Opacity, parameterName=Opacity Multiplier, sortPriority=1}
  ) {constB=2}
  n46 = ScalarParameter() {defaultValue=50, group=Opacity, parameterName=DepthFadeDistance}
  Multiply(A=@n28.A, B=ShadingPathSwitch(Default=DepthFade(Opacity=@n45, FadeDistance=@n46), Mobile=@n45) {desc=Migrated from FeatureLevelSwitch})
  n47 = DynamicParameter() {paramNames [Random,Dissolve,Param3,Param4]}
  n48 = AppendVector(A=@n47.Random, B=@n47.Random)
  n49 = VectorParameter() {group=Noise, parameterName=Noise B UV, sortPriority=6, defaultValue=<rgba 0.75 1 -1.5 -0.1>}
  n50 = TextureCoordinate()
  n51 = Panner(Coordinate=Multiply(A=@n50, B=AppendVector(A=@n49.R, B=@n49.R)), Speed=AppendVector(A=@n49.B, B=@n49.B))
  n52 = Multiply(A=@n9, B=ScalarParameter() {defaultValue=0.1, group=Distortion, parameterName=Distortion Intensity Noise, sortPriority=2})
  n53 = TextureSampleParameter2D(UVs=Add(A=Add(A=@n48, B=@n51), B=@n52)) {
   group=Noise
   parameterName=NoIse B Texture
   samplerType=SAMPLERTYPE_LinearColor
   sortPriority=4
   texture=/Game/Effects/Fort_Effects/Gameplay/SmokeDustExplosions/Textures/T_NoisePack_LinearColor
  }
  n54 = ChannelMaskParameter(None=@n53.RGB) {group=Noise, maskChannel=Green, parameterName=Noise B Channel, sortPriority=5, defaultValue=<rgba 0 1 0 0>}
  n55 = Saturate(None=Multiply(A=@n54.R, B=ScalarParameter() {defaultValue=1, group=Noise, parameterName=NoIse B Intensity, sortPriority=7}) {constB=10})
  n56 = AppendVector(A=@n47.Random, B=@n47.Random)
  n57 = VectorParameter() {group=Noise, parameterName=Noise A UV, sortPriority=2, defaultValue=<rgba 1 0.5 -1.5 0.2>}
  n58 = Panner(Coordinate=Multiply(A=@n50, B=AppendVector(A=@n57.R, B=@n57.R)), Speed=AppendVector(A=@n57.B, B=@n57.B))
  n59 = TextureSampleParameter2D(UVs=Add(A=@n52, B=Add(A=@n56, B=@n58))) {group=Noise, parameterName=NoIse A Texture, sortPriority=0, texture=&T_ShieldCaustics}
  n60 = Saturate(None=Multiply(
    A=ChannelMaskParameter(None=@n59.RGB) {group=Noise, parameterName=Noise A Channel, sortPriority=1}.R
    B=ScalarParameter() {defaultValue=1, group=Noise, parameterName=NoIse A Intensity, sortPriority=3}
   ) {constB=10})
  n61 = ScalarParameter() {defaultValue=0.35, group=Noise, parameterName=Lerp Noise AB, sortPriority=8}
  Multiply(
   A=LinearInterpolate(
    B=LinearInterpolate(A=@n55, B=@n60, Alpha=@n61)
    Alpha=ScalarParameter() {defaultValue=0.8, group=Noise, parameterName=Lerp Noise, sortPriority=8}
   ) {constA=1}
   B=@n27
  )
 ]
}
