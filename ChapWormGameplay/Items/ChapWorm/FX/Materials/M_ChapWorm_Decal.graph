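# Translucent unlit deferred decal material for ChapWorm effects, building an animated, noise-broken mask in world space.
# World position with a time-scrolled Z samples cloth-sim noise and Voronoi textures as 3D coordinates.
# A distance falloff around the object position, offset and scaled by a Size vector, forms the "3D Mask" with a noisy outer edge.
# Noise, decal lifetime opacity and parameters shape opacity, with Color and Normal-strength controls.
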Material /ChapWormGameplay/Items/ChapWorm/FX/Materials/M_ChapWorm_Decal
graph {
 properties {
  bEnableResponsiveAA=true
  bOutputTranslucentVelocity=true
  bUsedWithNiagaraRibbons=true
  bUsedWithNiagaraSprites=true
  blendMode=BLEND_Translucent
  materialDomain=MD_DeferredDecal
  pixelDepthOffsetMode=PDOM_Legacy
  shadingModel=MSM_Unlit
  twoSided=true
  usedShadingModels=Unlit
 }
 nodes [
  Constant()
  n1 = WorldPosition()
  n2 = Time()
  n3 = ScalarParameter() {defaultValue=-100, parameterName=Time Speed}
  n4 = AppendVector(A=@n1.XY, B=Add(A=@n1.Z, B=Multiply(A=@n2, B=@n3)))
  n5 = TextureSample(UVs=Multiply(A=@n4, B=ScalarParameter() {defaultValue=0.002, parameterName=Detail1}) {constB=0.003}) {samplerType=SAMPLERTYPE_LinearColor, texture=/Game/Global/MaterialFunctions/VT_ClothSimWPO}
  n6 = Power(Base=@n5.RGB, Exp=ScalarParameter() {defaultValue=1, parameterName=Noise Power})
  n7 = Saturate(None=Multiply(A=@n6, B=NamedRerouteUsage()))
  n8 = MaterialFunctionCall(Float2=DecalLifetimeOpacity().Opacity) {materialFunction=&BreakOutFloat2Components}
  Multiply(A=Multiply(A=@n6, B=Multiply(A=@n7, B=@n8.R)), B=ScalarParameter() {defaultValue=2, parameterName=Opacity})
  Multiply(A=@n1.XY) {constB=0.005}
  Multiply(
   A=MaterialFunctionCall(R=@n5.R, G=@n5.R) {materialFunction=&AppendMany}.RGB
   B=ScalarParameter() {defaultValue=1.5, parameterName=Normal}
  )
  n9 = ObjectPositionWS()
  n10 = NamedRerouteUsage()
  n11 = Multiply(A=Add(A=WorldPosition().XYZ, B=Multiply(A=@n10) {constB=30}), B=VectorParameter() {parameterName=Size, defaultValue=<rgba 1 1 2 0>}.RGB)
  n12 = Subtract(
   A=Multiply(A=Distance(A=@n9, B=@n11), B=ScalarParameter() {defaultValue=0.0004, parameterName=Range}) {constB=0.001}
   B=ScalarParameter() {defaultValue=0.02, parameterName=Decal Range Ctrl}
  ) {constB=0.02}
  n13 = Power(Base=OneMinus(None=Saturate(None=@n12))) {constExponent=150}
  n14 = Multiply(A=@n13) {constB=40}
  n15 = Multiply(A=@n13) {constB=600}
  n16 = Multiply(A=Multiply(A=Saturate(None=@n15), B=NamedRerouteUsage()), B=ScalarParameter() {defaultValue=1, parameterName=Out Noise Strength})
  NamedRerouteDeclaration(None=Saturate(None=Add(A=@n14, B=@n16))) {name=3D Mask, nodeColor=<rgba 0 0.6016 1 1>}
  n17 = TextureSample(UVs=Multiply(A=@n4, B=ScalarParameter() {defaultValue=0.01, parameterName=Detail2}) {constB=0.007}) {samplerType=SAMPLERTYPE_LinearColor, texture=/Game/Effects/Explosion_Update/Textures/VT_Voronoi_Seamless}
  n18 = ScalarParameter() {defaultValue=14, parameterName=Voronoi_Pow}
  NamedRerouteDeclaration(None=Multiply(A=Power(Base=@n17.R, Exp=@n18) {constExponent=4}, B=ScalarParameter() {defaultValue=5, parameterName=Voronoi_Strength}) {constB=2}) {name=Noise out, nodeColor=<rgba 0.7344 1 0 1>}
  NamedRerouteDeclaration(None=@n5.RGB) {name=Noise1, nodeColor=<rgba 0.8516 0 1 1>}
  ParticlePositionWS()
  Subtract(A=@n5.RGB) {constB=0.5}
  VectorParameter() {parameterName=Color, defaultValue=<rgba 0.3226 0.2006 0.8958 1>}
 ]
}
