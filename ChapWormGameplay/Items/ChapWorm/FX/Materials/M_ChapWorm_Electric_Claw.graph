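# Unlit translucent, two-sided Niagara sprite material for the ChapWorm electric claw effect, tinted purple by default.
# Emissive combines the claw texture's green channel with world-aligned screen-space smoke, multiplied by color parameters and particle color.
# "Use glow" fades out the smoke and blends opacity from doubled red to the blue channel.
# A Dissolve dynamic parameter drives a cloud-noise HeightLerp that, with particle alpha, masks opacity.
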Material /ChapWormGameplay/Items/ChapWorm/FX/Materials/M_ChapWorm_Electric_Claw
graph {
 properties {
  bUsedWithNiagaraSprites=true
  blendMode=BLEND_Translucent
  pixelDepthOffsetMode=PDOM_Legacy
  shadingModel=MSM_Unlit
  twoSided=true
  usedShadingModels=Unlit
 }
 nodes [
  n1 = TextureSample(UVs=MaterialFunctionCall(XY=TextureCoordinate()) {materialFunction=&Swizzle}.YX) {texture=/ChapWormGameplay/Items/ChapWorm/FX/Textures/T_ChapWorm_Claw_Default}
  n2 = LinearInterpolate(A=@n1.G, B=@n1.G, Alpha=ScalarParameter() {parameterName=Use glow})
  n3 = MaterialFunctionCall!(Center Point World Position) {materialFunction=/Engine/Functions/Engine_MaterialFunctions02/UVs/WorldSpaceAlignedScreenCoordinates}
  n4 = Multiply(A=TextureSample(UVs=Multiply(A=@n3.Offset and Scaled UVS) {constB=30}) {texture=/Game/Effects/Fort_Effects/GameplayPulginsAssets/Doom/DoomCastle/Smoke/MaterialAndMesh/Smoke01_D}.RGB) {constB=0.4}
  n5 = ScalarParameter() {parameterName=Use glow}
  n6 = Add(
   A=Multiply(A=VectorParameter() {parameterName=Color1, defaultValue=<rgba 0.7191 0 2.4 1>}.RGB, B=Multiply(A=@n2))
   B=Multiply(A=LinearInterpolate(A=@n4, Alpha=@n5) {constB=0})
  )
  n7 = ParticleColor()
  MaterialFunctionCall(Emissive=Multiply(A=VectorParameter() {parameterName=Color2, defaultValue=<rgba 0.5277 0 1.5 1>}.RGB, B=Multiply(A=@n6, B=@n7.RGB))) {materialFunction=&MF_FX_EmissiveAdjustments}
  n8 = Add(A=@n1.R, B=@n1.R)
  n9 = ScalarParameter() {parameterName=Use glow}
  n10 = MaterialFunctionCall(XY=TextureCoordinate() {vTiling=0.3}) {materialFunction=&Swizzle}
  n11 = Multiply(A=MaterialFunctionCall(
    Transition Phase=DynamicParameter() {paramNames [Dissolve,Param2,Param3,Param4]}.Dissolve
    Height Texture=TextureSample(UVs=@n10.YX) {samplerType=SAMPLERTYPE_LinearGrayscale, texture=/Game/Athena/Prototype/Terrain/T_CloudCoverage}.R
   ) {materialFunction=&HeightLerp}.Alpha) {constB=3}
  Multiply(A=@n7.A, B=Multiply(A=LinearInterpolate(A=@n8, B=@n1.B, Alpha=@n9), B=Saturate(None=@n11)))
 ]
}
