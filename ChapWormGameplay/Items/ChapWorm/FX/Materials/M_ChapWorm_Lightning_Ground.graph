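# Unlit, translucent, two-sided Niagara mesh-particle material for the ChapWorm ground lightning effect.
# World position offset blends SplineThicken with a pull toward a Dynamic Parameter XYZ position along UV V, plus a small UV-masked Y/Z offset.
# Panning curl noise bends the middle of the bolt, but PC low/medium/performance, Android and Switch skip it.
# Static switches select glow or particle color and a center-peaked core-and-halo opacity across U.
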Material /ChapWormGameplay/Items/ChapWorm/FX/Materials/M_ChapWorm_Lightning_Ground
graph {
 properties {
  bAutomaticallySetUsageInEditor=false
  bUsedWithNiagaraMeshParticles=true
  blendMode=BLEND_Translucent
  shadingModel=MSM_Unlit
  twoSided=true
  usedShadingModels=Unlit
 }
 nodes [
  n1 = VectorParameter() {parameterName=Thicken, channelNames {a=none, b=none, g=Tip, r=Base}}
  n2 = WorldPosition()
  n3 = Multiply(A=TextureCoordinate(), B=ScalarParameter() {defaultValue=16, parameterName=Param_1})
  n4 = Add(
   A=Subtract(A=WorldPosition().XYZ, B=ParticlePositionWS())
   B=Multiply(A=VectorParameter() {parameterName=Curl Noise Pan Dir, defaultValue=<rgba 50 100 -30 0>}.RGB, B=Time())
  )
  n5 = ScalarParameter() {defaultValue=0.0011, parameterName=Noise Size}
  n6 = Subtract(A=TextureSample(UVs=Multiply(A=@n4, B=@n5)) {samplerType=SAMPLERTYPE_LinearColor, texture=/Game/Effects/FogVolume/VT_CurlNoise_LOW}.RGB) {constB=0.5}
  n7 = DynamicParameter() {defaultValue=<rgba 100 100 200 10>, paramNames [X, Y, Z, WPO Noise Strength]}
  n8 = ComponentMask(None=TextureCoordinate()) {g=true}
  n9 = OneMinus(None=@n8)
  n10 = Multiply(A=@n9, B=@n8)
  n11 = NamedRerouteUsage()
  n12 = Subtract(
   A=Multiply(A=@n11, B=DynamicParameter() {defaultValue=<rgba 100 100 200 10>, paramNames [X, Y, Z, WPO Noise Strength]}.RGB)
   B=Multiply(A=@n11, B=@n2.XYZ)
  )
  n13 = Add(A=Multiply(A=Multiply(A=@n6, B=@n7.WPO Noise Strength), B=Multiply(A=@n10) {constB=2}), B=@n12)
  n14 = Constant3Vector() {constant=<rgba 0 1 1 0>}
  n15 = ComponentMask(None=TextureCoordinate()) {r=true}
  n16 = Abs(None=Multiply(A=Multiply(A=@n2.XYZ, B=@n14), B=Multiply(A=@n9, B=@n15)))
  Add(
   A=LinearInterpolate(
    A=MaterialFunctionCall(WidthBase=@n1.Base, WidthTip=@n1.Base, WorldPosition=@n2.XYZ, UVs for Thickness=@n3) {materialFunction=&SplineThicken}.WorldPositionOffset
    B=MaterialFunctionCall!("Xb1,PS4"=@n13, "PC-High&&Epic,XSX,PS5"=@n13, "PC-Perf,Android"=@n12, Switch=@n12, "PC-Low(NotPerf)&&Medium"=@n12) {materialFunction=/Game/Effects/Niagara/MasterMaterials/Functions/Common/PlatformSwitch_3D}.Result
    Alpha=NamedRerouteUsage()
   )
   B=Multiply(A=Saturate(None=@n16), B=ScalarParameter() {defaultValue=5})
  )
  NamedRerouteDeclaration(None=@n8) {name=WPO Mask, nodeColor=<rgba 1 0 0.1172 1>}
  n17 = ParticleColor()
  StaticSwitchParameter(True=VectorParameter() {parameterName=Glow Color, defaultValue=<rgba 0.3095 0 2 1>}.RGB, False=@n17.RGB) {parameterName=IS_Glow}
  n18 = ComponentMask(None=TextureCoordinate()) {r=true}
  n19 = OneMinus(None=@n18)
  n20 = Power(Base=Multiply(A=Multiply(A=@n19, B=@n18)) {constB=2}, Exp=ScalarParameter() {defaultValue=6.5, parameterName=Glow Pow})
  n21 = Subtract(A=@n20) {constB=0.01}
  n22 = Add(
   A=Multiply(A=Saturate(None=@n21), B=ScalarParameter() {defaultValue=10, parameterName=Core})
   B=Multiply(A=@n20, B=ScalarParameter() {defaultValue=0.4, parameterName=Glow Alpha})
  )
  StaticSwitchParameter(True=Multiply(A=@n17.A, B=@n22), False=StaticSwitchParameter(True=@n22, False=@n17.A) {parameterName=IS_Glow}) {parameterName=IS_Glow_Opacity}
 ]
}
