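# Niagara system that draws animation trails for the ChapWorm item, apparently claw swipes, with a stronger look when powered.
# One emitter spawns eight anim trails from skeletal mesh bone or socket indices, updates them against the mesh, and renders them through two ribbon renderers.
# User.IS_Powered switches claw colors, lifetime, color curves and ribbon widths.
# Trails fade and narrow over life, with primary RGB multiplied by a time-of-day scalar.
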NiagaraSystem /ChapWormGameplay/Items/ChapWorm/FX/NS_ChapWorm_AnimTrail
dependencies {
 usedDataInterfaces [&NiagaraDataInterfaceColorCurve, &NiagaraDataInterfaceCurve, &NiagaraDataInterfaceSkeletalMesh]
 usedDynamicInputs [&FNTimeOfDayScalar]
 usedModules [
  &self:SystemSpawnScript.SystemScriptSource.SystemScriptGraph.NiagaraNodeAssignment_1.SetVariables_#1
  &SystemState
  &self:SystemSpawnScript.SystemScriptSource.SystemScriptGraph.NiagaraNodeAssignment_0.SetVariables_#2
 ]
}
emitters {Trail001 {
  emitterUpdateScript [
   &EmitterState()
   &FN_SpawnAnimTrail(AnimTrailWidthMode=::FromFirst, Number of Anim Trails=8, Spawn Mode:ENiagaraSkeletalSpawning=::NewEnumerator0)
   /SetVariables(
    Emitter.Base Claw Color:LinearColor=&MakeCustomLinearColorFromBool(
     Bool=&SetBoolByIntComparison(A=$User.IS_Powered, B=1, Comparison Type=::NewEnumerator4)
     True Color=<rgba 0.318 0 1 1>
     False Color=<rgba 0.0913 0.0346 0.151 0.1>
    )
    Emitter.Second Claw Color:LinearColor=&MakeCustomLinearColorFromBool(
     Bool=&SetBoolByIntComparison(A=$User.IS_Powered, B=1, Comparison Type=::NewEnumerator4)
     True Color=<rgba 0.8347 0 5 1>
     False Color=<rgba 0.4494 0.0981 1 0.1>
    )
   ) {isSetParameters=true}
  ]
  particleSpawnScript [
   &FN_InitializeAnimTrailBoneIndecies(Second Bone or Socket Index=0)
   &FN_InitializeAnimTrailBoneIndecies(AnimTrail Selection=1, First Bone or Socket Index=1) {module=FN_InitializeAnimTrailBoneIndecies001}
   &FN_InitializeAnimTrailBoneIndecies(AnimTrail Selection=2, First Bone or Socket Index=2, Second Bone or Socket Index=2) {module=FN_InitializeAnimTrailBoneIndecies003}
   &FN_InitializeAnimTrailBoneIndecies(AnimTrail Selection=3, First Bone or Socket Index=3, Second Bone or Socket Index=3) {module=FN_InitializeAnimTrailBoneIndecies002}
   &FN_InitializeAnimTrailBoneIndecies(AnimTrail Selection=4, First Bone or Socket Index=4, Second Bone or Socket Index=4) {module=FN_InitializeAnimTrailBoneIndecies004}
   &FN_InitializeAnimTrailBoneIndecies(AnimTrail Selection=5, First Bone or Socket Index=5, Second Bone or Socket Index=5) {module=FN_InitializeAnimTrailBoneIndecies005}
   &FN_InitializeAnimTrailBoneIndecies(AnimTrail Selection=6, First Bone or Socket Index=6, Second Bone or Socket Index=6) {module=FN_InitializeAnimTrailBoneIndecies006}
   &FN_InitializeAnimTrailBoneIndecies(AnimTrail Selection=7, First Bone or Socket Index=7, Second Bone or Socket Index=7) {module=FN_InitializeAnimTrailBoneIndecies007}
   &FN_InitializeAnimTrail(
    Lifetime=&MakeCustomFloatFromBool(
     Bool=&SetBoolByIntComparison(A=$User.IS_Powered, B=1, Comparison Type=::NewEnumerator4)
     True Float=0.3
     False Float=0.28
    )
    SkeletalMesh=$System.SkeletalMesh
   )
  ]
  particleUpdateScript [
   &ParticleState()
   &FN_UpdateAnimTrail(SkeletalMesh=$System.SkeletalMesh)
   &ScaleColor(
    Scale Alpha=&FloatFromCurve(FloatCurve=DataInterface() {propertyValues {Curve {keys [
         {arriveTangent=-2.4978, interpMode=RCIM_Cubic, leaveTangent=-2.4978, tangentMode=RCTM_User, time=0, value=1}
         {interpMode=RCIM_Cubic, tangentMode=RCTM_User, time=0.6, value=0}
        ]}}})
    Scale RGB=&MakeCustomVectorFromBool(
     Bool=&SetBoolByIntComparison(A=$User.IS_Powered, B=1, Comparison Type=::NewEnumerator4)
     True Vector=&Multiply_VectorByFloat(
      Vector=&MakeVectorFromLinearColorRGB(Color=&ColorFromCurve(ColorCurve=DataInterface() {propertyValues {
          AlphaCurve {keys []}
          BlueCurve {keys [<at 0.0047 40>,<at 0.0515 20>,<at 0.3102 0.5 RCIM_Cubic>]}
          GreenCurve {keys [<at 0.0047 0>,<at 0.0515 0.8>,<at 0.3102 0 RCIM_Cubic>]}
          RedCurve {keys [<at 0.0047 30.6802>,<at 0.0515 4.3554>,<at 0.3102 0.3835 RCIM_Cubic>]}
         }}))
      Float=$System.TOD
     )
     False Vector=&Multiply_VectorByFloat(
      Vector=&MakeVectorFromLinearColorRGB(Color=&ColorFromCurve(ColorCurve=DataInterface() {propertyValues {
          AlphaCurve {keys []}
          BlueCurve {keys [<at 0.0047 40>,<at 0.1276 5>,<at 0.4431 0.5 RCIM_Cubic>]}
          GreenCurve {keys [<at 0.0047 0>,<at 0.1276 0.0602>,<at 0.4431 0 RCIM_Cubic>]}
          RedCurve {keys [<at 0.0047 30.6802>,<at 0.1276 1.1476>,<at 0.4431 0.3835 RCIM_Cubic>]}
         }}))
      Float=$System.TOD
     )
    )
    ScaleA=true
    ScaleRGB=true
   )
   &ScaleRibbonWidth(Ribbon Width Scale=&FloatFromCurve(
     FloatCurve=DataInterface() {propertyValues {Curve {keys [{arriveTangent=-0.0042, interpMode=RCIM_Cubic, leaveTangent=-0.0042, tangentMode=RCTM_User, time=0.1, value=1.1}, <at 1 0>]}}}
     Scale Curve=&MakeCustomFloatFromBool(
      Bool=&SetBoolByIntComparison(A=$User.IS_Powered, B=1, Comparison Type=::NewEnumerator4)
      True Float=3
      False Float=1.5
     )
    ))
   /SetVariables(
    Particles.Color2=&MakeCustomLinearColorFromBool(
     Bool=&SetBoolByIntComparison(A=$User.IS_Powered, B=1, Comparison Type=::NewEnumerator4)
     True Color=&ColorFromCurve(ColorCurve=DataInterface() {propertyValues {
        AlphaCurve {keys [<at 0 0.15 RCIM_Cubic>,<at 0.9 0 RCIM_Cubic>]}
        BlueCurve {keys [<at 0 25 RCIM_Cubic>,<at 0.7441 0 RCIM_Cubic>]}
        GreenCurve {keys [<at 0 0 RCIM_Cubic>,<at 0.7441 0 RCIM_Cubic>]}
        RedCurve {keys [<at 0 7.5788 RCIM_Cubic>,<at 0.7441 0 RCIM_Cubic>]}
       }})
     False Color=&ColorFromCurve(ColorCurve=DataInterface() {propertyValues {
        AlphaCurve {keys [<at 0 0.05 RCIM_Cubic>,<at 0.6157 0 RCIM_Cubic>]}
        BlueCurve {keys [<at 0 15 RCIM_Cubic>,<at 0.5715 0 RCIM_Cubic>]}
        GreenCurve {keys [<at 0 0 RCIM_Cubic>,<at 0.5715 0 RCIM_Cubic>]}
        RedCurve {keys [<at 0 4.0551 RCIM_Cubic>,<at 0.5715 0 RCIM_Cubic>]}
       }})
    )
    Particles.RibbonWidth2:NiagaraFloat=&MakeCustomFloatFromBool(
     Bool=&SetBoolByIntComparison(A=$User.IS_Powered, B=1, Comparison Type=::NewEnumerator4)
     True Float=14
     False Float=8
    )
   ) {isSetParameters=true}
  ]
  rendererClasses [&NiagaraRibbonRendererProperties]
  renderers [
   {rendererClass=&NiagaraRibbonRendererProperties}
   {rendererClass=&NiagaraRibbonRendererProperties, rendererIndex=1}
  ]
 }}
summary {
 systemName=NS_ChapWorm_AnimTrail
 emitters [{emitterName=Trail001, rendererClasses [&NiagaraRibbonRendererProperties]}]
 userVariables [
  User.Duration {type=NiagaraFloat, defaultValue=0.4, description=""}
  User.IS_Powered {type=NiagaraInt32}
  User.RibbonIDOffset {type=NiagaraInt32}
 ]
}
