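# Niagara effect for the hit impact of the ChapWorm ground punch, a short purple burst.
# An emitter with a garbled pasted-node name fires staggered mesh and sprite flashes, with flickering curves and animated material parameters.
# Decal_Burst adds a fading decal, while GPU Sparks_Up throws curl-noise sparks that shift to a color picked by User.ColorVariant.
# The Merged1 mesh emitter is disabled; User.Global Scale and User.IS_Fail have no visible use here.
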NiagaraSystem /ChapWormGameplay/Items/ChapWorm/FX/NS_ChapWorm_GroundPunch_Impact_Hit
dependencies {
 usedDataInterfaces [
  &NiagaraDataInterfaceCurve
  &NiagaraDataInterfaceColorCurve
  &NiagaraDataInterfaceArrayColor
  &NiagaraDataInterfaceArrayFloat
  &NiagaraDataInterfaceSkeletalMesh
 ]
 usedDynamicInputs []
 usedModules [
  &self:SystemSpawnScript.SystemScriptSource.SystemScriptGraph.NiagaraNodeAssignment_0.SetVariables_#1
  &SystemState
 ]
}
emitters {
 Begin_Object_Class__Script_NiagaraEditor_NiagaraOverviewNode_Name__NiagaraOverviewNode_19__ExportPath___Script_NiagaraEditor_NiagaraOverviewNode__ChapWormGameplay_Items_ChapWorm_FX_NS_ChapWorm_GroundPunch_Impact_Powered_NS_ChapWorm_GroundPunch_Impact_Powered_NiagaraSystemEditorData_0_NiagaraOverview_NiagaraOverviewNode_19_____OwningSystem___Script_Niagara_NiagaraSystem__ChapWormGameplay_Items_ChapWorm_FX_NS_ChapWorm_GroundPunch_Impact_Powered_NS_ChapWorm_GroundPunch_Impact_Powered_____EmitterHandleGuid_#2___NodePosX_1152___NodePosY_-128___bUserSetEnabledState_True___NodeGuid_C41CBF004C8F7357399A8B874F199D9FEnd_Object {
  emitterUpdateScript [
   &EmitterState()
   &SpawnBurst_Instantaneous(Use Spawn Probability=false, UseLoopCountLimit=false)
   &SpawnBurst_Instantaneous(Spawn Count=2, Spawn Time=0.0542, Use Spawn Probability=false, UseLoopCountLimit=false) {module=SpawnBurst_Instantaneous001}
   &SpawnBurst_Instantaneous(Spawn Time=0.0958, Use Spawn Probability=false, UseLoopCountLimit=false) {module=SpawnBurst_Instantaneous002}
   &SpawnBurst_Instantaneous(Spawn Time=0.2, Use Spawn Probability=false, UseLoopCountLimit=false) {module=SpawnBurst_Instantaneous003}
  ]
  particleSpawnScript [
   /SetVariables(Particles.BiasedRandom=&FloatFromCurve(
     FloatCurve=DataInterface() {propertyValues {Curve {keys [
         {interpMode=RCIM_Cubic, tangentMode=RCTM_User, time=0, value=0}
         {arriveTangent=2.5197, interpMode=RCIM_Cubic, leaveTangent=2.5197, tangentMode=RCTM_User, time=0.6, value=0.6}
         {interpMode=RCIM_Cubic, tangentMode=RCTM_User, time=1, value=1}
        ]}}}
     CurveIndex=&RandomRangeFloat(Minimum=0, Maximum=1)
    )) {isSetParameters=true}
   /Niagara/Modules/Spawn/Initialization/InitializeParticle(
    Color=<rgba 0.367 0 0.5 1>
    Lifetime=&Lerp_Float(A=0.2, B=0.25, Alpha=$Particles.BiasedRandom)
    Mesh Scale=<0.4125 0.4125 -0.4125>
    Module.Write Color=true
    Module.Write Lifetime=true
    Module.Write Mass=false
    Module.Write Material Random=true
    Module.Write Position=true
    Module.Write Scale=true
    Module.Write SpriteRotation=false
    Module.Write SpriteSize=true
    Sprite Size=&Vector2DFromFloat(Value=&RandomRangeFloat(Minimum=350, Maximum=450))
   ) {enabled=false}
   /Niagara/Modules/Spawn/Initialization/V2/InitializeParticle(
    Color=<rgba 0.367 0 0.5 1>
    Lifetime=&Lerp_Float(A=0.15, B=0.2, Alpha=$Particles.BiasedRandom)
    Mesh Scale=<0.4125 0.4125 -0.4125>
    Mesh Scale Mode=::NewEnumerator4
    Position Offset=<0 10 70>
    Position Offset Coordinate Space=::Local
    Sprite Size=&Vector2DFromFloat(Value=&RandomRangeFloat(Minimum=350, Maximum=450))
    Sprite Size Mode=::NewEnumerator4
    UsePositionOffset=true
    Write Material Random=true
   ) {module=InitializeParticle001}
   &InitialMeshOrientation(Rotation=&RandomRangeVector(Minimum=<-0.25 0 -0.25>, Maximum=<0.25 0 0.25>), Use Rotation Vector=false) {module=InitialMeshOrientation001}
   /SetVariables() {isSetParameters=true}
   /SetVariables(Particles.MeshIndex=&RandomRangeInt(Minimum=0, Maximum=1)) {isSetParameters=true}
   /SetVariables(Particles.DynParamsSprite:Vector4f=<0.25 2 1 1>) {isSetParameters=true}
  ]
  particleUpdateScript [
   &ParticleState()
   &ScaleColor(
    Scale Alpha=&FloatFromCurve(FloatCurve=DataInterface() {propertyValues {Curve {keys [
         <at 0 0>
         {interpMode=RCIM_Cubic, tangentMode=RCTM_User, time=0.1258, value=0.1952}
         <at 0.341 0.0902>
         {interpMode=RCIM_Cubic, tangentMode=RCTM_User, time=0.496, value=0.2532}
         <at 1 0>
        ]}}})
    Scale RGB=&VectorFromFloat(Value=&FloatFromCurve(FloatCurve=DataInterface() {propertyValues {Curve {keys [
          {arriveTangent=3.5209, interpMode=RCIM_Cubic, leaveTangent=3.5209, tangentMode=RCTM_User, time=0, value=0.5}
          {arriveTangent=-20.2701, interpMode=RCIM_Cubic, leaveTangent=-20.2701, tangentMode=RCTM_User, time=0.45, value=5}
          <at 1 1>
         ]}}}))
    ScaleA=true
    ScaleRGB=true
   )
   /SetVariables(Particles.SpriteColor=&MultiplyLinearColor(
     Linear Color=$Particles.Color
     Scale Factor=&MakeVector4_FromLinearColor(LinearColor=&MultiplyLinearColor(
       Linear Color=&MultiplyLinearColor(Linear Color=<rgba 0.2 1 0.9 1>, Scale Factor=&MakeVector4_FromFloat(Float=1.5))
       Scale Factor=&MakeVector4_FromFloat(Float=&Multiply_Float(
         A=1
         B=&FloatFromCurve(FloatCurve=DataInterface() {propertyValues {Curve {keys [
              {interpMode=RCIM_Cubic, tangentMode=RCTM_User, time=0, value=1}
              {arriveTangentWeight=0.031, interpMode=RCIM_Cubic, leaveTangentWeight=0.0366, tangentMode=RCTM_User, time=0.0918, value=0.05}
              {arriveTangentWeight=0.16, interpMode=RCIM_Cubic, leaveTangent=18.7153, leaveTangentWeight=0.1479, tangentMode=RCTM_Break, time=0.18, value=0}
              {arriveTangent=-0.2815, interpMode=RCIM_Cubic, leaveTangent=-0.2815, tangentMode=RCTM_User, time=0.3, value=0.95}
              {arriveTangentWeight=0.0345, interpMode=RCIM_Cubic, leaveTangentWeight=0.0566, tangentMode=RCTM_User, time=0.4, value=0.1}
              {
               arriveTangentWeight=0.1589
               interpMode=RCIM_Cubic
               leaveTangent=29.6956
               leaveTangentWeight=0.1018
               tangentMode=RCTM_Break
               time=0.5
               value=0.05
              }
              {arriveTangent=0.512, interpMode=RCIM_Cubic, leaveTangent=0.512, tangentMode=RCTM_User, time=0.5641, value=0.8648}
              <at 1 0>
             ]}}})
        ))
      ))
    )) {isSetParameters=true}
   &DynamicMaterialParameters(
    Index 0 Param 1=&FloatFromCurve(FloatCurve=DataInterface() {propertyValues {Curve {keys [{tangentMode=RCTM_User, time=0, value=1}, {tangentMode=RCTM_User, time=1, value=0}]}}})
    Index 0 Param 2=&Add_Float(
     A=&FloatFromCurve(FloatCurve=DataInterface() {propertyValues {Curve {keys [
          {interpMode=RCIM_Cubic, leaveTangent=5.6571, tangentMode=RCTM_Break, time=0, value=0}
          {arriveTangent=0.0136, interpMode=RCIM_Cubic, leaveTangent=-5.3647, tangentMode=RCTM_Break, time=0.4833, value=1}
          <at 1 -0.1>
         ]}}})
     B=&Multiply_Float(A=$Particles.BiasedRandom, B=0.5)
    )
    Index 0 Param 3=&RandomRangeFloat(Minimum=0, Maximum=0.25)
    Index 1 Param 1=&Add_Float(A=$Particles.NormalizedAge, B=&RandomRangeFloat(Minimum=0, Maximum=1))
    Index 1 Param 2=&RandomRangeFloat(Minimum=-25, Maximum=25)
    Index 1 Param 3=3
    Module.Param0WriteEnabled=true
    Module.Param0WriteEnabled1=true
    Module.Param1WriteEnabled=true
    Module.Param1WriteEnabled1=true
    Module.Param2WriteEnabled=true
    Module.Param2WriteEnabled1=true
    Module.Param3WriteEnabled=false
    Module.Param3WriteEnabled1=false
    Write Parameter Index 1=true
   ) {module=DynamicMaterialParameters001}
   &ScaleSpriteSize(Uniform Curve Sprite Scale=DataInterface() {propertyValues {Curve {keys [
        {arriveTangent=21.5526, interpMode=RCIM_Cubic, time=0, value=0}
        <at 0.1 1 RCIM_Cubic>
        {arriveTangent=-1.8758, interpMode=RCIM_Cubic, leaveTangent=-1.8758, tangentMode=RCTM_User, time=1, value=0}
       ]}}})
   &ScaleColor(Scale RGB=&VectorFromFloat(Value=1), ScaleA=true, ScaleRGB=true) {module=ScaleColor001}
  ]
  rendererClasses [&NiagaraMeshRendererProperties, &NiagaraSpriteRendererProperties]
  renderers [
   {rendererClass=&NiagaraMeshRendererProperties, rendererIndex=0}
   {rendererClass=&NiagaraSpriteRendererProperties, rendererIndex=1}
  ]
 }
 Decal_Burst {
  emitterUpdateScript [
   &EmitterState(Life Cycle Mode=::NewEnumerator1, Loop Behavior=::NewEnumerator1, Loop Duration=0.1, UseLoopDelay=false)
   &SpawnBurst_Instantaneous(Use Spawn Probability=false, UseLoopCountLimit=false)
   /SetVariables(Emitter.Decal Color:LinearColor=&FN_Select_Color_from_Time_Of_Day(Morning=<rgba 0.3811 0.048 4 1>, Day=<rgba 0.3811 0.048 4 1>, Evening=<rgba 0.3811 0.048 4 1>, Night=<rgba 0.6553 0 4 1>)) {enabled=false, isSetParameters=true}
  ]
  particleSpawnScript [
   &InitializeParticle(
    Color=<rgba 1.05 0.06 6 4>
    Lifetime=0.6
    Mass Mode=::NewEnumerator1
    Position Offset=<0 0 100>
    Position Offset Coordinate Space=::World
    Sprite Rotation Mode=::NewEnumerator2
    Sprite Size Mode=::NewEnumerator3
    Sprite UV Mode=::NewEnumerator3
    Uniform Sprite Size=5
    UsePositionOffset=true
    Write Material Random=true
   )
   /SetVariables(Particles.DecalInitScale=1) {isSetParameters=true}
  ]
  particleUpdateScript [
   &ParticleState()
   /SetVariables(
    Particles.DecalFade:NiagaraFloat=&FloatFromCurve(FloatCurve=DataInterface() {propertyValues {Curve {keys [
         {arriveTangent=9.0725, interpMode=RCIM_Cubic, leaveTangent=9.0725, tangentMode=RCTM_User, time=0, value=0.1}
         {arriveTangent=-0.873, interpMode=RCIM_Cubic, leaveTangent=-0.873, tangentMode=RCTM_User, time=0.1, value=1}
         {arriveTangent=-1.1111, interpMode=RCIM_Cubic, leaveTangent=-1.1111, time=0.6, value=0.75}
         {arriveTangent=-0.0405, interpMode=RCIM_Cubic, leaveTangent=-0.0405, tangentMode=RCTM_User, time=1, value=0}
        ]}}})
    Particles.DecalSize:Vector3f=&Multiply_VectorByFloat(
     Vector=&Multiply_VectorByFloat(Vector=<350 600 600>, Float=1)
     Float=&FloatFromCurve(
      FloatCurve=DataInterface() {propertyValues {Curve {keys [
          {arriveTangent=9.608, interpMode=RCIM_Cubic, leaveTangent=9.608, tangentMode=RCTM_User, time=0, value=0}
          {arriveTangent=0.0224, interpMode=RCIM_Cubic, leaveTangent=0.0224, tangentMode=RCTM_User, time=0.01, value=1}
         ]}}}
      Scale Curve=$Particles.DecalInitScale
     )
    )
   ) {isSetParameters=true}
  ]
  rendererClasses [&NiagaraDecalRendererProperties]
  renderers [{rendererClass=&NiagaraDecalRendererProperties, rendererIndex=0}]
 }
 Merged1 {
  bEnabled=false
  emitterUpdateScript [
   &EmitterState()
   &SpawnBurst_Instantaneous(Spawn Count=1, Use Spawn Probability=false)
   &SpawnBurst_Instantaneous(Spawn Group=1, Use Spawn Probability=false, UseLoopCountLimit=false) {module=SpawnBurst_Instantaneous001}
   &SpawnBurst_Instantaneous(Spawn Group=2, Use Spawn Probability=false, UseLoopCountLimit=false) {enabled=false, module=SpawnBurst_Instantaneous002}
   /SetVariables(
    Emitter.Color:LinearColor=&ColorFromCurve(
     ColorCurve=DataInterface() {propertyValues {
       AlphaCurve {keys [<at 0 0>,<at 0.0924 1 RCIM_Cubic>,<at 0.3264 0 RCIM_Cubic>]}
       BlueCurve {keys [<at 0.0339 40>,<at 0.3334 0 RCIM_Cubic>]}
       GreenCurve {keys [<at 0.0339 0>,<at 0.3334 0 RCIM_Cubic>]}
       RedCurve {keys [<at 0.0339 8.1441>,<at 0.3334 0 RCIM_Cubic>]}
      }}
     CurveIndex=$Emitter.Age
    )
    Emitter.Position:NiagaraPosition=<0 0 100>
    Emitter.Radius:NiagaraFloat=&FloatFromCurve(
     FloatCurve=DataInterface() {propertyValues {Curve {keys [
         {tangentMode=RCTM_User, time=0, value=1}
         {arriveTangent=-0.0405, interpMode=RCIM_Cubic, leaveTangent=-0.0405, tangentMode=RCTM_User, time=0.4, value=1}
         {arriveTangent=-8.0089, interpMode=RCIM_Cubic, leaveTangent=-8.0089, tangentMode=RCTM_User, time=0.5, value=0}
        ]}}}
     CurveIndex=$Emitter.NormalizedLoopAge
     Scale Curve=420
    )
   ) {isSetParameters=true}
  ]
  particleSpawnScript [
   /SetVariables(Particles.VisibilityTag=$Emitter.SpawnGroup) {isSetParameters=true}
   &InitializeParticle(
    Color=<rgba 1.9378 0 15 0.2>
    Lifetime=&SelectFloatFromArray(Direct Array Index=$Emitter.SpawnGroup, Float Selection Array=DataInterface() {propertyValues {FloatData [0.5,0.4,0.3]}})
    Mesh Scale Mode=::NewEnumerator3
    Mesh Uniform Scale=12
    Position Offset=<0 10 90>
    Position Offset Coordinate Space=::Local
    Sprite Size=<800 500>
    Sprite Size Mode=::NewEnumerator4
    UsePositionOffset=true
    Write Material Random=true
   )
   &InitialMeshOrientation(
    Rotation=&RandomRangeVector(Minimum=<0 0 0>, Maximum=<1 0 0>)
    Use Orientation Vector=true
    Use Rotation Vector=true
   )
   /SetVariables(Particles.Color2=<rgba 50 37.4589 6.8714 2>) {isSetParameters=true}
  ]
  particleUpdateScript [
   &ParticleState()
   &ScaleSpriteSize(Uniform Curve Sprite Scale=DataInterface() {propertyValues {Curve {keys [
        {arriveTangent=4.1785, interpMode=RCIM_Cubic, leaveTangent=4.1785, tangentMode=RCTM_User, time=0, value=0.1}
        {arriveTangent=1.1179, interpMode=RCIM_Cubic, leaveTangent=1.1179, tangentMode=RCTM_User, time=0.2042, value=0.6228}
        {arriveTangent=0.0224, interpMode=RCIM_Cubic, leaveTangent=0.0224, tangentMode=RCTM_User, time=1, value=1}
       ]}}})
   &ScaleColor(
    Scale Alpha=&FloatFromCurve(
     FloatCurve=DataInterface() {propertyValues {Curve {keys [
         {arriveTangent=19.9472, interpMode=RCIM_Cubic, leaveTangent=19.9472, tangentMode=RCTM_User, time=0, value=0}
         <at 0.3 1.75 RCIM_Cubic>
         <at 1 0 RCIM_Cubic>
        ]}}}
     Scale Curve=&FNTimeOfDayScalar(Night Scale=0.4, Morning Scale=0.5, Evening Scale=0.5)
    )
    Scale RGB=<8 8 8>
    ScaleA=true
    ScaleRGB=true
   )
   &ScaleMeshSize(Scale Factor=&VectorFromFloat(Value=&FloatFromCurve(
      FloatCurve=DataInterface() {propertyValues {Curve {keys [
          {arriveTangent=9.2279, interpMode=RCIM_Cubic, leaveTangent=9.2279, tangentMode=RCTM_User, time=0, value=0.35}
          {arriveTangent=2.2956, interpMode=RCIM_Cubic, leaveTangent=2.2956, tangentMode=RCTM_User, time=0.3, value=0.2}
          {arriveTangent=-0.0405, interpMode=RCIM_Cubic, leaveTangent=-0.0405, tangentMode=RCTM_User, time=0.602, value=0.7688}
         ]}}}
      Scale Curve=0.85
     )))
   &DynamicMaterialParameters(
    Index 0 Param 1=&FloatFromCurve(FloatCurve=DataInterface() {propertyValues {Curve {keys [
         {arriveTangent=-0.3119, interpMode=RCIM_Cubic, leaveTangent=-0.3119, tangentMode=RCTM_User, time=0, value=0.9}
         {interpMode=RCIM_Cubic, tangentMode=RCTM_User, time=0.2, value=0.8}
         {arriveTangent=0.5138, interpMode=RCIM_Cubic, leaveTangent=0.5138, tangentMode=RCTM_User, time=0.5, value=0.8875}
         {interpMode=RCIM_Cubic, tangentMode=RCTM_User, time=1, value=1}
        ]}}})
    Index 0 Param 2=&FloatFromCurve(FloatCurve=DataInterface() {propertyValues {Curve {keys [<at 0 15 RCIM_Cubic>, {arriveTangent=-40.0295, interpMode=RCIM_Cubic, leaveTangent=-40.0295, tangentMode=RCTM_User, time=1, value=0}]}}})
    Index 0 Param 4=5
    Module.Param0WriteEnabled=true
    Module.Param1WriteEnabled=true
    Module.Param2WriteEnabled=true
    Module.Param3WriteEnabled=true
   )
   /SetVariables(
    Particles.Color1:LinearColor=&ColorFromCurve(ColorCurve=DataInterface() {propertyValues {
       AlphaCurve {keys [<at 0.003 0>, {arriveTangent=1.0168, interpMode=RCIM_Cubic, leaveTangent=1.0168, time=0.2427, value=2}, <at 0.4948 0.5>, <at 1 0 RCIM_Cubic>]}
       BlueCurve {keys [<at 0.2325 1 RCIM_Cubic>,<at 0.4828 2 RCIM_Cubic>]}
       GreenCurve {keys [<at 0.2325 0 RCIM_Cubic>,<at 0.4828 0 RCIM_Cubic>]}
       RedCurve {keys [<at 0.2325 0.221 RCIM_Cubic>,<at 0.4828 0.4365 RCIM_Cubic>]}
      }})
    Particles.Color2=&MultiplyLinearColor(
     Linear Color=$Particles.Color2
     Scale Factor=&Vector4FromCurve(Vector4Curve=DataInterface() {propertyValues {
        AlphaCurve {keys [<at 0 0>,<at 0.0939 0.75>,<at 0.6927 1>,<at 1 0>]}
        BlueCurve {keys [<at 0.3637 1 RCIM_Cubic>,<at 1 0.1878 RCIM_Cubic>]}
        GreenCurve {keys [<at 0.3637 0 RCIM_Cubic>,<at 1 0 RCIM_Cubic>]}
        RedCurve {keys [<at 0.3637 0.5245 RCIM_Cubic>,<at 1 0.0343 RCIM_Cubic>]}
       }})
    )
    Particles.DMP0A:Vector4f=&MakeVector4(
     X=0
     Y=&FloatFromCurve(FloatCurve=DataInterface() {propertyValues {Curve {keys [
          {interpMode=RCIM_Cubic, tangentMode=RCTM_User, time=0.5, value=0.25}
          {arriveTangent=-0.5125, interpMode=RCIM_Cubic, leaveTangent=-0.5125, tangentMode=RCTM_User, time=1, value=-1}
         ]}}})
     Z=1
     W=50
    )
    Particles.DMP2A:Vector4f=<1 1 1 1>
    Particles.MeshSize1:Vector3f=&Multiply_VectorByFloat(
     Vector=<8 8 8>
     Float=&FloatFromCurve(
      FloatCurve=DataInterface() {propertyValues {Curve {keys [
          {arriveTangent=4.1785, interpMode=RCIM_Cubic, leaveTangent=4.1785, tangentMode=RCTM_User, time=0, value=0.2}
          {arriveTangent=1.9532, interpMode=RCIM_Cubic, leaveTangent=1.9532, tangentMode=RCTM_User, time=0.2, value=0.7}
          {arriveTangent=-0.0405, interpMode=RCIM_Cubic, leaveTangent=-0.0405, tangentMode=RCTM_User, time=1, value=1.25}
         ]}}}
      Scale Curve=0.6
     )
    )
   ) {isSetParameters=true}
   &SkeletalMeshLocation(
    MaskSpawnGroup=false
    OverrideRandomSeed=false
    Sampled Position Offset=<0 0 100>
    Skeletal Mesh=DataInterface() {propertyValues {
      DefaultMesh=null
      ExcludeBoneName=root
      OptionalComponentClass=null
      PreviewMesh=/Game/Characters/Player/Male/Medium/Base/SK_M_MALE_Base_Skeleton
      SkinningMode=SkinOnTheFly
      SoftSourceActor=null
      SourceMode=Default
      UvSetIndex=0
      WholeMeshLOD=-1
      bExcludeBone=false
      bReadDeformedGeometry=true
      bRequireCurrentFrameData=true
      ComponentTags []
      FilteredBones []
      FilteredSockets []
      MeshUserParameter {parameter {name=None, typeDefHandle {registeredTypeIndex=1}, varData []}}
      SamplingRegions []
     }}
   ) {enabled=false}
  ]
  rendererClasses [&NiagaraMeshRendererProperties]
  renderers [
   {rendererClass=&NiagaraMeshRendererProperties, rendererIndex=0}
   {rendererClass=&NiagaraMeshRendererProperties, rendererIndex=1}
  ]
 }
 Sparks_Up {
  simTarget=GPUComputeSim
  emitterUpdateScript [
   &EmitterState(Life Cycle Mode=::NewEnumerator1, Loop Behavior=::NewEnumerator1, Loop Delay=0.0048, Loop Duration=1, UseLoopDelay=true)
   &SpawnBurst_Instantaneous(Spawn Count=90, Use Spawn Probability=false)
  ]
  particleSpawnScript [
   &InitializeParticle(
    Color=<rgba 10 10 10 1>
    Lifetime Max=0.3
    Lifetime Min=0.2
    Lifetime Mode=::NewEnumerator1
    Sprite Size Max=<8 20>
    Sprite Size Min=<15 10>
    Sprite Size Mode=::NewEnumerator2
    UsePositionOffset=false
    Write Material Random=true
   ) {module=InitializeParticle001}
   &ShapeLocation(MaskSpawnGroup=false, Offset Mode=::NewEnumerator0, Override Seed=false, Rotation Mode=::NewEnumerator1, UseOwnerScale=true)
   &AddVelocity(Module.Use Falloff Distance=false, Override Seed=false, Velocity Mode=::NewEnumerator1, Velocity Speed=500)
  ]
  particleUpdateScript [
   &ParticleState()
   &CameraOffset(Camera Offset Amount=100) {module=CameraOffset001}
   &CurlNoiseForce(
    Noise Frequency=5
    Noise Quality / Cost=::NewEnumerator0
    Noise Strength=&Multiply_Float(
     A=800
     B=&FloatFromCurve(FloatCurve=DataInterface() {propertyValues {Curve {keys [<at 0 0>,<at 0.1 1>,<at 1 1.5>]}}})
    )
    Pan Noise=false
   )
   &GravityForce(Gravity=&RandomRangeVector(Minimum=<0 0 2000>, Maximum=<0 0 -2000>))
   &Drag(Drag=&UniformRangedFloat(Minimum=7, Maximum=11), Use Linear Drag=true, Use Rotational Drag=false)
   &SolveForcesAndVelocity(Clamp Velocity=false, Limit Acceleration=false)
   &ScaleSpriteSize(Scale Factor=&Vector2DFromFloat(Value=&FloatFromCurve(FloatCurve=DataInterface() {propertyValues {Curve {keys [<at 0.4 1>,<at 1 0>]}}}, Scale Curve=0.3)))
   &ScaleSpriteSizeBySpeed(Max Scale Factor=<0.3 2>, Min Scale Factor=<0.7 0.5>)
   &ScaleColor(
    Scale RGB=&Lerp_Vector(
     A=&MakeVectorFromLinearColorRGB(Color=&ColorFromCurve(ColorCurve=DataInterface() {propertyValues {
         AlphaCurve {keys []}
         BlueCurve {keys [<at 0 8>,<at 0.3699 0>]}
         GreenCurve {keys [<at 0 0.1454>,<at 0.3699 0>]}
         RedCurve {keys [<at 0 1.8104>,<at 0.3699 0>]}
        }}))
     B=&MakeVectorFromLinearColorRGB(Color=&SelectLinearColorFromArray(
       Array Sampling Mode=::NewEnumerator1
       Direct Array Index=$User.ColorVariant
       Color Selection Array=DataInterface() {propertyValues {ColorData [<rgba 0 0.4025 1.5 1>, <rgba 1 0 0.0169 1>, <rgba 1 0.1618 0 1>, <rgba 0.3522 0 1 1>]}}
       Random Seed=?
      ))
     Alpha=&FloatFromCurve(FloatCurve=DataInterface() {propertyValues {Curve {keys [
          {arriveTangent=4.1785, interpMode=RCIM_Cubic, leaveTangent=4.1785, tangentMode=RCTM_User, time=0.15, value=0}
          {arriveTangent=2.0911, interpMode=RCIM_Cubic, leaveTangent=2.0911, tangentMode=RCTM_User, time=0.35, value=0.6228}
          {arriveTangent=0.0224, interpMode=RCIM_Cubic, leaveTangent=0.0224, tangentMode=RCTM_User, time=0.7, value=1}
         ]}}})
    )
    ScaleA=true
    ScaleRGB=true
   )
  ]
  rendererClasses [&NiagaraSpriteRendererProperties]
  renderers [{rendererClass=&NiagaraSpriteRendererProperties}]
 }
}
summary {
 systemName=NS_ChapWorm_GroundPunch_Impact_Hit
 emitters [
  {emitterName=Sparks_Up, simTarget=GPUComputeSim, rendererClasses [&NiagaraSpriteRendererProperties]}
  {bEnabled=false, emitterName=Merged1, rendererClasses [&NiagaraMeshRendererProperties]}
  {
   emitterName=Begin_Object_Class__Script_NiagaraEditor_NiagaraOverviewNode_Name__NiagaraOverviewNode_19__ExportPath___Script_NiagaraEditor_NiagaraOverviewNode__ChapWormGameplay_Items_ChapWorm_FX_NS_ChapWorm_GroundPunch_Impact_Powered_NS_ChapWorm_GroundPunch_Impact_Powered_NiagaraSystemEditorData_0_NiagaraOverview_NiagaraOverviewNode_19_____OwningSystem___Script_Niagara_NiagaraSystem__ChapWormGameplay_Items_ChapWorm_FX_NS_ChapWorm_GroundPunch_Impact_Powered_NS_ChapWorm_GroundPunch_Impact_Powered_____EmitterHandleGuid_#2___NodePosX_1152___NodePosY_-128___bUserSetEnabledState_True___NodeGuid_C41CBF004C8F7357399A8B874F199D9FEnd_Object
   rendererClasses [&NiagaraMeshRendererProperties, &NiagaraSpriteRendererProperties]
  }
  {emitterName=Decal_Burst, rendererClasses [&NiagaraDecalRendererProperties]}
 ]
 userVariables [
  User.ColorVariant {type=NiagaraInt32, defaultValue=3}
  User.Global Scale {type=NiagaraFloat, defaultValue=1, description=""}
  User.IS_Fail {type=NiagaraBool, defaultValue=true}
 ]
}
