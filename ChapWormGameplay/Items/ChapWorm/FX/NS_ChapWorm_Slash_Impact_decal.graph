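# Niagara system for the ChapWorm slash impact decal, a short-lived burst drawn with a decal renderer.
# A single Decal_Burst emitter spawns an instantaneous burst, offsetting particles upward in world space.
# Curves snap decal size to full almost immediately while fade rises quickly, then eases to zero over the half-second lifetime.
# An emitter-level decal color is selected by time of day, a purple-blue HDR tint that shifts slightly at night.
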NiagaraSystem /ChapWormGameplay/Items/ChapWorm/FX/NS_ChapWorm_Slash_Impact_decal
dependencies {
 usedDataInterfaces [&NiagaraDataInterfaceCurve]
 usedDynamicInputs [&UniformRangedFloat]
 usedModules [
  &self:SystemSpawnScript.SystemScriptSource.SystemScriptGraph.NiagaraNodeAssignment_0.SetVariables_#1
  &SystemState
 ]
}
emitters {Decal_Burst {
  emitterUpdateScript [
   &EmitterState(Life Cycle Mode=::NewEnumerator1, Loop Behavior=::NewEnumerator1, Loop Duration=0.1, UseLoopDelay=false)
   &SpawnBurst_Instantaneous(Use Spawn Probability=false, UseLoopCountLimit=false)
   /SetVariables(Emitter.Decal Color:LinearColor=&FN_Select_Color_from_Time_Of_Day(Morning=<rgba 0.3811 0.048 4 1>, Day=<rgba 0.3811 0.048 4 1>, Evening=<rgba 0.3811 0.048 4 1>, Night=<rgba 0.6553 0 4 1>)) {isSetParameters=true}
  ]
  particleSpawnScript [
   &InitializeParticle(
    Color Mode=::NewEnumerator0
    Lifetime=0.5
    Mass Mode=::NewEnumerator1
    Position Offset=<0 0 110>
    Position Offset Coordinate Space=::World
    Sprite Rotation Mode=::NewEnumerator2
    Sprite Size Mode=::NewEnumerator3
    Sprite UV Mode=::NewEnumerator3
    Uniform Sprite Size=5
    UsePositionOffset=true
    Write Material Random=true
   )
   /SetVariables(Particles.DecalInitScale=1) {isSetParameters=true}
  ]
  particleUpdateScript [
   &ParticleState()
   /SetVariables(
    Particles.DecalFade:NiagaraFloat=&FloatFromCurve(FloatCurve=DataInterface() {propertyValues {Curve {keys [
         {arriveTangent=9.0725, interpMode=RCIM_Cubic, leaveTangent=9.0725, tangentMode=RCTM_User, time=0, value=0.1}
         {arriveTangent=-0.873, interpMode=RCIM_Cubic, leaveTangent=-0.873, tangentMode=RCTM_User, time=0.1, value=0.55}
         {arriveTangent=-0.6111, interpMode=RCIM_Cubic, leaveTangent=-0.6111, time=0.5, value=0.2}
         {arriveTangent=-0.0405, interpMode=RCIM_Cubic, leaveTangent=-0.0405, tangentMode=RCTM_User, time=1, value=0}
        ]}}})
    Particles.DecalSize:Vector3f=&Multiply_VectorByFloat(
     Vector=&Multiply_VectorByFloat(Vector=<250 400 400>, Float=1)
     Float=&FloatFromCurve(
      FloatCurve=DataInterface() {propertyValues {Curve {keys [
          {arriveTangent=9.608, interpMode=RCIM_Cubic, leaveTangent=9.608, tangentMode=RCTM_User, time=0, value=0}
          {arriveTangent=0.0224, interpMode=RCIM_Cubic, leaveTangent=0.0224, tangentMode=RCTM_User, time=0.01, value=1}
         ]}}}
      Scale Curve=$Particles.DecalInitScale
     )
    )
   ) {isSetParameters=true}
  ]
  rendererClasses [&NiagaraDecalRendererProperties]
  renderers [{rendererClass=&NiagaraDecalRendererProperties, rendererIndex=0}]
 }}
summary {
 systemName=NS_ChapWorm_Slash_Impact_decal
 emitters [{emitterName=Decal_Burst, rendererClasses [&NiagaraDecalRendererProperties]}]
 userVariables [
  User.ColorVariant {type=NiagaraInt32, defaultValue=1}
  User.Global Scale {type=NiagaraFloat, defaultValue=1, description=""}
 ]
}
