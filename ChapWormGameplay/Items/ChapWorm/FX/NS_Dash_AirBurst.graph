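# Niagara one-shot air burst effect, apparently for the ChapWorm dash, with emitter loop duration set by User.uDuration.
# A single burst emitter packs suffixed attribute sets for wind cone, air glow and glow corona into short-lived particles.
# Curves animate their color, alpha and size over life, feeding two mesh renderers.
# Corona and waveform-flickered light attributes are also computed, but no sprite or light renderer is present.
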NiagaraSystem /ChapWormGameplay/Items/ChapWorm/FX/NS_Dash_AirBurst
dependencies {
 usedDataInterfaces [&NiagaraDataInterfaceColorCurve, &NiagaraDataInterfaceCurve]
 usedDynamicInputs []
 usedModules [&SystemState]
}
emitters {Combined_Burst_Once {
  emitterUpdateScript [
   &EmitterState(
    Inactive Response=::NewEnumerator1
    Life Cycle Mode=::NewEnumerator1
    Loop Behavior=::NewEnumerator1
    Loop Duration=$User.uDuration
    UseLoopDelay=false
   )
   &SpawnBurst_Instantaneous(Use Spawn Probability=false, UseLoopCountLimit=false)
  ]
  particleSpawnScript [
   &InitializeParticle(Color Mode=::NewEnumerator0, Lifetime=0.4, Position Mode=::NewEnumerator0, Write Material Random=true)
   /SetVariables(
    Particles.Color_WindCone:LinearColor=<rgba 0.4583 0.4032 0.3558 1.5>
    Particles.DynamicParameter_WindCone:Vector4f=<0.7 2.5 0 0>
    Particles.Orientation_WindCone:Quat4f=&OrientMeshToVector(Look At Direction=<1 0 0>, Reference Axis Vector=<1 0 0>)
    Particles.Size_WindCone:Vector3f=<0.8 1.2 2>
   ) {isSetParameters=true}
   /SetVariables(
    Particles.Color_AirGlow:LinearColor=<rgba 0.125 0.125 0.125 1.5>
    Particles.DynamicParameter_AirGlow:Vector4f=<1 1 0 1>
    Particles.Orientation_AirGlow:Quat4f=&OrientMeshToVector(Look At Direction=<1 0 0>, Reference Axis Vector=<1 0 0>)
    Particles.Size_AirGlow:Vector3f=&Multiply_VectorByFloat(Vector=<2 2.25 2.25>, Float=0.8)
   ) {isSetParameters=true}
   /SetVariables(
    Particles.Color_GlowCorona:LinearColor=<rgba 1 0.1667 0 0.5>
    Particles.Position_GlowCorona:NiagaraPosition=<-5 0 0>
    Particles.Size_GlowCorona:Vector2f=&Multiply_Vector2D_ByFloat(Float=125, Vector2D=<1.5 0.8>)
   ) {isSetParameters=true}
   /SetVariables(Particles.DynamicParameter_Sprite:Vector4f=<1 1 1 1>) {isSetParameters=true}
   /SetVariables(Particles.CameraOffsetGlowCorona=15) {isSetParameters=true}
  ]
  particleUpdateScript [
   &ParticleState()
   /SetVariables(
    Particles.Color_WindCone:LinearColor=&MultiplyLinearColor(
     Linear Color=&ColorFromCurve(ColorCurve=DataInterface() {propertyValues {
        AlphaCurve {keys [<at 0 1 RCIM_Cubic>]}
        BlueCurve {keys [<at 0.2953 5 RCIM_Cubic>,<at 0.4027 0.9858 RCIM_Cubic>]}
        GreenCurve {keys [<at 0.2953 0 RCIM_Cubic>,<at 0.4027 0.8583 RCIM_Cubic>]}
        RedCurve {keys [<at 0.2953 0.5668 RCIM_Cubic>,<at 0.4027 1 RCIM_Cubic>]}
       }})
     Scale Factor=&MakeFloat4FromVectorAndFloat(
      "Vector (RGB)"=<1 1 1>
      "Float (Alpha)"=&FloatFromCurve(
       FloatCurve=DataInterface() {propertyValues {Curve {keys [
           {arriveTangent=11.4472, interpMode=RCIM_Cubic, leaveTangent=11.4472, tangentMode=RCTM_User, time=0, value=0}
           {arriveTangent=1.1179, interpMode=RCIM_Cubic, leaveTangent=1.1179, tangentMode=RCTM_User, time=0.1, value=0.8}
           {arriveTangent=0.0224, interpMode=RCIM_Cubic, leaveTangent=0.0224, tangentMode=RCTM_User, time=1, value=0}
          ]}}}
       Scale Curve=2
      )
     )
    )
    Particles.Size_WindCone:Vector3f=&Multiply_Vector(
     A=$Particles.Initial.Size_WindCone
     B=&MakeVector(
      X=1
      Y=&FloatFromCurve(
       FloatCurve=DataInterface() {propertyValues {Curve {keys [{tangentMode=RCTM_User, time=0, value=0}, {tangentMode=RCTM_User, time=1, value=1}]}}}
       Scale Curve=2
      )
      Z=&FloatFromCurve(
       FloatCurve=DataInterface() {propertyValues {Curve {keys [{tangentMode=RCTM_User, time=0, value=0}, {tangentMode=RCTM_User, time=1, value=1}]}}}
       Scale Curve=2
      )
     )
    )
   ) {isSetParameters=true}
   /SetVariables(Particles.Color_AirGlow:LinearColor=&MultiplyLinearColor(
     Linear Color=$Particles.Initial.Color_AirGlow
     Scale Factor=&MakeFloat4FromVectorAndFloat(
      "Vector (RGB)"=&VectorFromFloat(Value=5)
      "Float (Alpha)"=&FloatFromCurve(FloatCurve=DataInterface() {propertyValues {Curve {keys [
           {arriveTangent=11.4472, interpMode=RCIM_Cubic, leaveTangent=11.4472, tangentMode=RCTM_User, time=0, value=0}
           {arriveTangent=1.1179, interpMode=RCIM_Cubic, leaveTangent=1.1179, tangentMode=RCTM_User, time=0.2, value=0.75}
           {arriveTangent=0.0224, interpMode=RCIM_Cubic, leaveTangent=0.0224, tangentMode=RCTM_User, time=1, value=0}
          ]}}})
     )
    )) {isSetParameters=true}
   /SetVariables(
    Particles.Color_GlowCorona:LinearColor=&MultiplyLinearColor(
     Linear Color=$Particles.Initial.Color_GlowCorona
     Scale Factor=&MakeFloat4FromVectorAndFloat(
      "Vector (RGB)"=&VectorFromFloat(Value=&FloatFromCurve(
        FloatCurve=DataInterface() {propertyValues {Curve {keys [
            {arriveTangent=47.5885, interpMode=RCIM_Cubic, time=0, value=0}
            {arriveTangent=2.8125, interpMode=RCIM_Cubic, leaveTangent=2.8125, time=0.1, value=5}
            <at 0.8 2.25>
            <at 1 0 RCIM_Cubic>
           ]}}}
        Scale Curve=15
       ))
      "Float (Alpha)"=&FloatFromCurve(FloatCurve=DataInterface() {propertyValues {Curve {keys [
           <at 0 1 RCIM_Cubic>
           {arriveTangent=-1.3333, interpMode=RCIM_Cubic, leaveTangent=-1.3333, time=0.1, value=1.25}
           {arriveTangent=-1.3889, interpMode=RCIM_Cubic, leaveTangent=-1.3889, time=0.3, value=0.6}
           <at 1 0 RCIM_Cubic>
          ]}}})
     )
    )
    Particles.Size_GlowCorona:Vector2f=&Multiply_Vector2D_ByFloat(
     Float=&FloatFromCurve(
      FloatCurve=DataInterface() {propertyValues {Curve {keys [
          {arriveTangent=0.225, interpMode=RCIM_Cubic, time=0, value=0.025}
          {arriveTangent=1.2812, interpMode=RCIM_Cubic, leaveTangent=1.2812, time=0.1, value=0.575}
          {arriveTangent=-0.6389, interpMode=RCIM_Cubic, leaveTangent=-0.6389, time=0.8, value=1.05}
          {arriveTangent=-11.4662, interpMode=RCIM_Cubic, leaveTangent=-11.4662, tangentMode=RCTM_User, time=1, value=0}
         ]}}}
      Scale Curve=5
     )
     Vector2D=$Particles.Initial.Size_GlowCorona
    )
   ) {isSetParameters=true}
   &FN_Light_Attributes(
    Color=<rgba 1 0.0833 0 1>
    Create Color Attribute=true
    Create Exponent Attribute=true
    Create Radius Attribute=true
    Radius=100
    Scale Color=&Multiply_Float(
     A=$System.Interpolation_Thrust_EndLife
     B=&Waveform("Waveform [1]"=::NewEnumerator0, "[1] Frequency"=5, "[1] Amplitude Min/Max"=<0.5 1>)
    )
   )
  ]
  rendererClasses [&NiagaraMeshRendererProperties]
  renderers [
   {rendererClass=&NiagaraMeshRendererProperties, rendererIndex=0}
   {rendererClass=&NiagaraMeshRendererProperties, rendererIndex=1}
  ]
 }}
summary {
 systemName=NS_Dash_AirBurst
 emitters [{emitterName=Combined_Burst_Once, rendererClasses [&NiagaraMeshRendererProperties]}]
 userVariables [User.uDuration {type=NiagaraFloat, defaultValue=1.5, description=""}]
}
