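# Material function that layers reactive, charge-driven glow onto ChapWorm item materials, apparently a gauntlet.
# A mask texture drives base color tint and emissive details, while a second mask input supplies pulse, overpowered, energy timeline and charge signals.
# Two time-offset pulses through a noise-distorted sphere mask in pre-skinned space animate a Fresnel-lit hex pattern, alongside a panning caustics glow.
# Summed emissive passes through EyeAdaptationInverse before output.
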MaterialFunction /ChapWormGameplay/Items/ChapWorm/Materials/MF_ChapWorm_Reactive
graph {nodes [
  n1 = Fresnel() {baseReflectFraction=0, exponent=1}
  n2 = Power(Base=@n1, Exp=ScalarParameter() {defaultValue=3, group=SnapFreeze_Reactivity, parameterName=SFG_FresnelPower})
  n3 = TextureSampleParameter2D(UVs=Multiply(A=TextureCoordinate(), B=ScalarParameter() {defaultValue=24, group=SnapFreeze_Reactivity, parameterName=SFG_Tiling})) {
   group=SnapFreeze_Reactivity
   parameterName=IdlePattern
   samplerType=SAMPLERTYPE_Masks
   texture=/ChapWormGameplay/Items/ChapWorm/Textures/T_PlumHexPattern
  }
  n4 = Multiply(
   A=LinearInterpolate(
    B=ScalarParameter() {defaultValue=3, group=SnapFreeze_Reactivity, parameterName=SFG_IdleFresnelBrightness}
    Alpha=NamedRerouteDeclaration(None=@n2) {name=Fresnel, nodeColor=<rgba 0.5703 1 0 1>}
   ) {constA=1, constB=3}
   B=Multiply(A=@n3.R, B=NamedRerouteUsage())
  )
  n5 = VectorParameter() {group=SnapFreeze_Reactivity, parameterName=IdlePatternColorLow, defaultValue=<rgba 0.02 0 0.3 1>}
  n6 = VectorParameter() {group=SnapFreeze_Reactivity, parameterName=IdlePatternColorHigh, defaultValue=<rgba 0.3662 0.0493 1 1>}
  n7 = Multiply(A=NamedRerouteUsage(), B=Multiply(A=@n4, B=LinearInterpolate(A=@n5.RGB, B=@n6.RGB, Alpha=@n4))) {constA=1}
  Add(A=@n7, B=NamedRerouteUsage())
  n8 = GetMaterialAttributes(None=FunctionInput() {inputName=Material, inputType=FunctionInput_MaterialAttributes}) {attributeGetTypes [#1, #2]}
  n9 = VectorParameter() {group=SnapFreeze_Reactivity, parameterName=GlowingDetails_BaseColor, sortPriority=50, defaultValue=<rgba 0.2044 0 0.276 1>}
  n10 = VectorParameter() {group=SnapFreeze_Reactivity, parameterName=GlowingDetails_BaseColor_Rim, sortPriority=50, defaultValue=<rgba 0.2044 0 0.276 1>}
  n11 = Power(Base=NamedRerouteUsage(), Exp=ScalarParameter() {defaultValue=0.2, group=SnapFreeze_Reactivity, parameterName=GlowingDetails_FresnelExp}) {constExponent=0.2}
  n12 = LinearInterpolate(A=@n9.RGB, B=@n10.RGB, Alpha=@n11)
  n13 = TextureSampleParameter2D() {
   group=SnapFreeze_Reactivity
   parameterName=ReactiveMask
   samplerSource=SSM_Wrap_WorldGroupSettings
   samplerType=SAMPLERTYPE_Masks
   sortPriority=0
   texture=/ChapWormGameplay/Items/ChapWorm/Textures/T_ChapWorm_Gauntlet_GM
  }
  n14 = MaterialFunctionCall(Float4=FunctionInput(Preview=AppendVector(A=@n13.RGB, B=@n13.RGB)) {bUsePreviewValueAsDefault=true, inputName=Masks, inputType=FunctionInput_Vector4}) {materialFunction=&BreakOutFloat4Components}
  n15 = NamedRerouteDeclaration(None=@n14.R) {name=GM_Red, nodeColor=<rgba 0.0078 1 0 1>}
  n16 = LinearInterpolate(A=@n8.BaseColor, B=@n12, Alpha=@n15)
  n17 = Add(A=@n8.BaseColor, B=Multiply(A=@n15, B=@n12))
  n18 = VectorParameter() {group=SnapFreeze_Reactivity, parameterName=GlowingDetails_EmissiveColor, sortPriority=50, defaultValue=<rgba 0.0942 0 0.709 1>}
  n19 = VectorParameter() {group=SnapFreeze_Reactivity, parameterName=GlowingDetails_EmissiveColor_Rim, sortPriority=50, defaultValue=<rgba 0.0942 0 0.709 1>}
  n20 = Multiply(
   A=Multiply(
    A=NamedRerouteDeclaration(None=@n14.B) {name=GM_Blue, nodeColor=<rgba 0.0078 1 0 1>}
    B=LinearInterpolate(A=@n18.RGB, B=@n19.RGB, Alpha=@n11)
   )
   B=ScalarParameter() {defaultValue=1, group=SnapFreeze_Reactivity, parameterName=GlowingDetails_Intensity, sortPriority=51}
  )
  n21 = GetMaterialAttributes(None=SetMaterialAttributes!(
    MaterialAttributes=@n8.MaterialAttributes
    BaseColor=StaticSwitchParameter(True=@n16, False=@n17) {group=SnapFreeze_Reactivity, parameterName=GlowingDetails_ReplaceBaseColor}
    EmissiveColor=Add(A=@n8.EmissiveColor, B=@n20)
   ) {attributeSetTypes [#1, #2]}) {attributeGetTypes [#2, #3, #1]}
  n22 = Add(A=@n21.EmissiveColor, B=NamedRerouteUsage())
  n23 = NamedRerouteUsage()
  n24 = SetMaterialAttributes!(
   MaterialAttributes=@n21.MaterialAttributes
   EmissiveColor=EyeAdaptationInverse(LightValueInput=Add(A=@n22, B=@n23)).EyeAdaptationInverse
   BaseColor
   Metallic
   Specular
   Roughness
  ) {attributeSetTypes [
    #2
    #1
    #4
    #5
    #6
   ]}
  FunctionOutput(None=@n24)
  n25 = Time()
  n26 = ScalarParameter() {defaultValue=0.15, group=SnapFreeze_Reactivity, parameterName=SFG_PulseSpeed}
  n27 = Multiply(A=@n25, B=@n26)
  n28 = NamedRerouteUsage()
  n29 = Multiply(
   A=ScalarParameter() {defaultValue=0.1, group=SnapFreeze_Reactivity, parameterName=SFG_IdlePulseWidth}
   B=Max(A=NamedRerouteUsage(), B=Saturate(None=@n28))
  )
  n30 = NamedRerouteUsage()
  n31 = NamedRerouteUsage()
  n32 = Panner(
   Coordinate=Divide(A=ComponentMask(None=@n30) {b=true, r=true}) {constB=64}
   Speed=LinearInterpolate(A=Constant2Vector() {g=-0.027, r=-0.021}, B=Constant2Vector() {g=0.54, r=-0.42}, Alpha=Saturate(None=@n31))
  ) {speedX=-0.021, speedY=-0.027}
  n33 = NamedRerouteUsage()
  n34 = Multiply(A=NamedRerouteDeclaration(None=TextureSample(UVs=@n32, Tex=@n33) {samplerSource=SSM_Wrap_WorldGroupSettings, samplerType=SAMPLERTYPE_LinearColor}.R) {name=PS-Noise1, nodeColor=<rgba 1 0.2812 0 1>}) {constB=0.5}
  n35 = TextureObject() {samplerType=SAMPLERTYPE_Masks, texture=/Game/Global/Textures/Default/Noise/T_Noise_Cloud_Mask}
  n36 = LocalPosition() {localOrigin=InstancePreSkinning}
  n37 = FunctionInput(Preview=ShaderStageSwitch(PixelShader=VertexInterpolator(VS=@n36.XYZ).PS, VertexShader=@n36.XYZ)) {bUsePreviewValueAsDefault=true, inputName=PreskinnedPosition}
  n38 = Power(
   Base=MaterialFunctionCall(
    TextureObject=NamedRerouteDeclaration(None=@n35) {name=NoiseTex, nodeColor=<rgba 0.1484 0 1 1>}
    TextureSize=VectorParameter() {group=SnapFreeze_Reactivity, parameterName=SFG_PulseNoiseScale, defaultValue=<rgba 512 512 512 1>}.RGB
    WorldPosition=NamedRerouteDeclaration(None=@n37) {name=PS-LocalPos, nodeColor=<rgba 0 0.1562 1 1>}
   ) {materialFunction=&WorldAlignedTexture}.XYZ Texture
   Exp=ScalarParameter() {defaultValue=3, group=SnapFreeze_Reactivity, parameterName=SFG_WorldAlignedNoisePower}
  ) {constExponent=3}
  n39 = Subtract(A=Multiply(A=Add(A=@n34, B=@n38)) {constB=2})
  n40 = NamedRerouteUsage()
  n41 = Normalize(VectorInput=@n40)
  n42 = Multiply(
   A=Subtract(A=NamedRerouteDeclaration(None=@n3.A) {name=Random, nodeColor=<rgba 0.6641 1 0 1>}) {constB=0.5}
   B=ScalarParameter() {defaultValue=15, group=SnapFreeze_Reactivity, parameterName=IdlePatternOffset}
  )
  n43 = SphereMask(
   A=Add(
    A=Multiply(A=@n39) {constB=64}
    B=VectorParameter() {group=SnapFreeze_Reactivity, parameterName=SFG_MaskPosition, defaultValue=<rgba 0 8 128 1>}.RGB
   )
   B=Add(A=@n40, B=Multiply(A=@n41, B=@n42))
   Radius=ScalarParameter() {defaultValue=192, group=SnapFreeze_Reactivity, parameterName=SFG_HitRadius}
   Hardness=ScalarParameter() {group=SnapFreeze_Reactivity, parameterName=SFG_MaskHardness}
  )
  n44 = SphereMask(A=@n43, B=OneMinus(None=Frac(None=@n27)), Radius=@n29) {hardnessPercent=0}
  n45 = Multiply(A=Add(A=@n25) {constB=2.355}, B=@n26)
  n46 = SphereMask(A=@n43, B=OneMinus(None=Frac(None=@n45)), Radius=@n29) {hardnessPercent=0}
  NamedRerouteDeclaration(None=SmoothStep(Value=Max(A=@n44, B=@n46))) {name=PulseNoise, nodeColor=<rgba 1 0.0234 0 1>}
  n47 = ScalarParameter() {defaultValue=5, group=SnapFreeze_Reactivity, parameterName=SFG_PatternBrightness}
  n48 = ScalarParameter() {defaultValue=15, group=SnapFreeze_Reactivity, parameterName=SFG_ChargePatternBrightness}
  n49 = NamedRerouteUsage()
  NamedRerouteDeclaration(None=Multiply(A=@n7, B=LinearInterpolate(A=@n47, B=@n48, Alpha=@n49)) {constB=7}) {name=PatternColor, nodeColor=<rgba 0 0.1797 1 1>}
  n50 = MaterialFunctionCall(Float4=FunctionInput() {bUsePreviewValueAsDefault=true, inputName=Masks 2, inputType=FunctionInput_Vector4}) {materialFunction=&BreakOutFloat4Components}
  NamedRerouteDeclaration(None=@n50.R) {name=Pulse Animation Mask, nodeColor=<rgba 1 0 0.8672 1>}
  NamedRerouteDeclaration(None=@n50.G) {name=Overpowered State, nodeColor=<rgba 0.6406 0 1 1>}
  n51 = ScalarParameter() {defaultValue=0.8, parameterName=EmissiveAlbedoBlend}
  n52 = PixelNormalWS()
  n53 = MaterialFunctionCall(
   Brightness=ScalarParameter() {defaultValue=0.6, parameterName=FresnelBrightness}
   Exponent=ScalarParameter() {defaultValue=2.5, parameterName=FresnelExponent}
   Vertex / Pixel Normal=@n52
  ) {materialFunction=&MF_Fresnel_FX}
  n54 = ScalarParameter() {defaultValue=0.5, parameterName=FresnelThickness}
  n55 = ScalarParameter() {defaultValue=1.5, parameterName=NormalizeFresnelExponent}
  n56 = Saturate(None=Max(
    A=@n53.Fresnel Inverted
    B=MaterialFunctionCall!(Thickness=@n54, NormalWS=@n52, PowerExponent=@n55) {materialFunction=/Game/Weapons/Prototype/ObjectMover/Materials/MF_NormalizedFresnel}.PowerInverted
   ))
  n57 = VectorParameter() {parameterName=Panning Glow Color, defaultValue=<rgba 0.02 0 0.3 1>}
  n58 = VertexInterpolator(VS=LocalPosition() {localOrigin=InstancePreSkinning}.XYZ)
  n59 = Sign(None=ComponentMask(None=@n58.PS) {r=true})
  n60 = Constant() {r=1}
  n61 = Multiply(
   A=Time()
   B=Multiply(
    A=VectorParameter() {parameterName=PanSpeedPSLP, defaultValue=<rgba -5 0 -20 0>}.RGB
    B=MaterialFunctionCall(R=@n59, G=@n60, B=@n60) {materialFunction=&AppendMany}.RGB
   )
  )
  n62 = PreSkinnedNormal()
  n63 = Multiply(
   A=MaterialFunctionCall(
    TextureObject=TextureObject() {samplerType=SAMPLERTYPE_LinearColor, texture=/Game/Weapons/FORT_Melee/Pickaxe_RoyalDsuk/Textures/T_Pickaxe_RoyalDusk_Caustics}
    TextureSize=VectorParameter() {parameterName=CausticsScale, defaultValue=<rgba 20 20 40 0>}.RGB
    WorldPosition=Subtract(A=@n58.PS, B=@n61)
    World Space Normal=VertexInterpolator(VS=@n62).PS
   ) {materialFunction=&WorldAlignedTexture}.XYZ Texture
   B=VectorParameter() {parameterName=CausticColor, defaultValue=<rgba 0.3662 0.0493 1 1>}.RGB
  )
  n64 = ComponentMask(None=@n58.PS) {b=true}
  n65 = Saturate(None=Divide(A=Add(A=@n64) {constB=-95}) {constB=25})
  n66 = Multiply(A=Add(A=Multiply(A=@n56, B=@n57.RGB), B=Multiply(A=@n53.Fresnel Inverted, B=@n63)), B=Power(Base=OneMinus(None=@n65)))
  n67 = ScalarParameter() {defaultValue=14, parameterName=GlowActive}
  n68 = LinearInterpolate(
   A=Multiply(A=@n67, B=ScalarParameter() {defaultValue=0.45, parameterName=GlowActiveLowChargeMultiplier}) {constB=0.75}
   B=@n67
   Alpha=NamedRerouteUsage()
  )
  n69 = NamedRerouteUsage()
  n70 = NamedRerouteUsage()
  n71 = Add(A=Multiply(A=Max(A=@n69, B=@n70)) {constB=5})
  NamedRerouteDeclaration(None=Multiply(A=Multiply(A=@n51, B=@n66), B=Multiply(A=@n68, B=@n71))) {name=Fresnel Glow, nodeColor=<rgba 0.9219 0 1 1>}
  n72 = Multiply(
   A=ScalarParameter() {defaultValue=0.75, group=SnapFreeze_Reactivity, parameterName=SFG_PatternFresnelStrength}
   B=Saturate(None=NamedRerouteUsage())
  )
  n73 = ScalarParameter() {defaultValue=2, parameterName=SFG_PatternFresnelExponent}
  n74 = ScalarParameter() {parameterName=SFG_PatternFresnelBase}
  NamedRerouteDeclaration(None=Multiply(A=Multiply(A=@n5.RGB, B=@n72), B=Fresnel(ExponentIn=@n73, BaseReflectFractionIn=@n74))) {name=PatternFresnel, nodeColor=<rgba 0 1 0.7891 1>}
  NamedRerouteDeclaration(None=Power(Base=@n50.A)) {name=Charge Amount, nodeColor=<rgba 1 1 1 1>}
  NamedRerouteDeclaration(None=@n50.B) {name=Energy Timeline, nodeColor=<rgba 1 0 0.4922 1>}
  NamedRerouteDeclaration(None=@n14.G) {name=GM_Green, nodeColor=<rgba 1 0 0.0937 1>}
  OneMinus(None=@n1)
 ]}
