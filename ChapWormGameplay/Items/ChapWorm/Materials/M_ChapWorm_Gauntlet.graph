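# Surface material for the ChapWorm gauntlet, adding charge, attack and overpower effects through MF_ChapWorm_Reactive on top of the 2023 character parent PreFX and PostFX functions.
# Custom primitive data supplies charge level, toggles and start times, which become pulse, energy and reversible overpower timelines.
# A pre-skinned sphere mask at the hand, side-selected by a Left/Right value, localizes attack energy.
# Texture masks and animated values are packed into float4 inputs.
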Material /ChapWormGameplay/Items/ChapWorm/Materials/M_ChapWorm_Gauntlet
graph {
 properties {
  bUseMaterialAttributes=true
  bUsedWithClothing=true
  bUsedWithMeshDeformer=true
  bUsedWithNanite=true
  bUsedWithNiagaraMeshParticles=true
  bUsedWithSkeletalMesh=true
  pixelDepthOffsetMode=PDOM_Legacy
  previewMesh=/ChapWormGameplay/Items/ChapWorm/Meshes/ChapWorm_Gauntlet
  shadingModel=MSM_FromMaterialExpression
  usedShadingModels=Default Lit | Subsurface | Clear Coat | Subsurface Profile | Cloth
 }
 nodes [
  n1 = MaterialFunctionCall() {materialFunction=&MF_CharacterParent_2023_PreFX}
  n2 = Power(Base=NamedRerouteUsage())
  MaterialFunctionCall!(
   PreskinnedPosition
   Masks=NamedRerouteUsage()
   Masks 2=MaterialFunctionCall(X=Sine(None=@n2) {period=2}, Y=NamedRerouteUsage(), Z=NamedRerouteUsage(), A=NamedRerouteUsage()) {materialFunction=&MakeFloat4}.Result
   Material=MaterialFunctionCall(Material=@n1.Result, M=@n1.Result, Pre Skinned Local Position=@n1.Result, Distance Blend=@n1.Result, Fresnel=@n1.Result) {materialFunction=&MF_CharacterParent_2023_PostFX}.Result
  ) {materialFunction=&/MF_ChapWorm_Reactive}
  n3 = ScalarParameter() {bUseCustomPrimitiveData=true, parameterName=Left/Right, primitiveDataIndex=16}
  n4 = Constant() {r=1}
  n5 = SphereMask(
   A=NamedRerouteUsage()
   B=Multiply(
    A=VectorParameter() {parameterName=Hnad Location, defaultValue=<rgba 66 15 83 1>}.RGB
    B=MaterialFunctionCall(X=@n3, Y=@n4, Z=@n4) {materialFunction=&MakeFloat3}.Result
   )
  ) {attenuationRadius=30, hardnessPercent=0}
  n6 = Subtract(A=Time(), B=ScalarParameter() {bUseCustomPrimitiveData=true, parameterName=EnergyTime, primitiveDataIndex=12})
  n7 = ScalarParameter() {defaultValue=0.15, parameterName=Dissipation Time}
  n8 = OneMinus(None=Saturate(None=Divide(A=@n6, B=@n7)))
  n9 = Multiply(A=Multiply(A=NamedRerouteDeclaration(None=@n5) {name=Attack Side, nodeColor=<rgba 0.1016 1 0 1>}) {constB=6}, B=@n8)
  NamedRerouteDeclaration(None=LinearInterpolate(A=Add(A=@n9, B=@n8), Alpha=ScalarParameter() {bUseCustomPrimitiveData=true, parameterName=Toggle Sticky Charging, primitiveDataIndex=13})) {name=Energy Timeline, nodeColor=<rgba 0.5968 1 0 1>}
  n10 = Time()
  n11 = ScalarParameter() {bUseCustomPrimitiveData=true, parameterName=Overpower Anim, primitiveDataIndex=14}
  n12 = Saturate(None=Divide(A=Subtract(A=@n10, B=@n11), B=ScalarParameter() {defaultValue=0.5, parameterName=Overpower Transition Time}))
  NamedRerouteDeclaration(None=LinearInterpolate(
    A=OneMinus(None=@n12)
    B=@n12
    Alpha=ScalarParameter() {bUseCustomPrimitiveData=true, parameterName=Forward Direction, primitiveDataIndex=15}
   )) {name=Overpower State, nodeColor=<rgba 1 0.0469 0 1>}
  n13 = TextureSample() {samplerType=SAMPLERTYPE_Masks, texture=/ChapWormGameplay/Items/ChapWorm/Textures/T_ChapWorm_Gauntlet_GM}
  n14 = Max(A=NamedRerouteUsage(), B=NamedRerouteUsage())
  n15 = TextureSample() {samplerType=SAMPLERTYPE_Masks, texture=/ChapWormGameplay/Items/ChapWorm/Textures/T_ChapWorm_Gauntlet_FX}
  n16 = Multiply(A=SmoothStep(Value=OneMinus(None=@n15.R)) {constMin=0.2}, B=OneMinus(None=@n13.R))
  NamedRerouteDeclaration(None=MaterialFunctionCall(X=Multiply(A=@n13.R, B=@n14), Y=NamedRerouteDeclaration(None=@n16) {name=Emissive Mask, nodeColor=<rgba 1 0 0.5391 1>}, Z=@n13.B, A=@n14) {materialFunction=&MakeFloat4}.Result) {name=Function mask, nodeColor=<rgba 0.8516 0 1 1>}
  n17 = LocalPosition() {localOrigin=InstancePreSkinning}
  NamedRerouteDeclaration(None=ShaderStageSwitch(PixelShader=VertexInterpolator(VS=@n17.XYZ).PS, VertexShader=@n17.XYZ)) {name=Preskinned, nodeColor=<rgba 0.2187 1 0 1>}
  n18 = ScalarParameter() {bUseCustomPrimitiveData=true, parameterName=Current Charge, primitiveDataIndex=10}
  NamedRerouteDeclaration(None=@n18) {name=Current Charge Normalized, nodeColor=<rgba 1 0 0.7031 1>}
  n19 = Divide(
   A=Subtract(A=Time(), B=ScalarParameter() {bUseCustomPrimitiveData=true, parameterName=AnimTime, primitiveDataIndex=11})
   B=ScalarParameter() {defaultValue=0.8, parameterName=Duration}
  )
  NamedRerouteDeclaration(None=OneMinus(None=Saturate(None=@n19))) {name=Pulse Timeline, nodeColor=<rgba 1 0.0469 0 1>}
  Power(Base=@n18)
  VertexInterpolator(VS=LocalPosition() {localOrigin=InstancePreSkinning}.XYZ)
 ]
}
